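; Weapon Blueprint for the PointSmoke pickaxe cosmetic that tells its swing particle effect which montage section is playing.
; Left and right swing events each run the parent logic, then derive a section index from the instigator through a utility.
; That index goes to the swing VFX component as the SwingSectionIndex integer parameter, or -1 when the utility's cast fails.
; BeginPlay and Tick are empty, and the construction script only calls its parent.
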(event EventSwingLeft
  (bind _returnvalue (CallFunction|GetSwingVFXComponent))
  (|CallParentFunction)
  (bind _section_index (Utilities|SetupSwingMontageSectionIndex (Variables|Actor|GetInstigator))
    (:Out
      (Effects|Components|ParticleSystem|SetIntParameter _returnvalue "SwingSectionIndex" _section_index))
    (:Cast%20Failed%20Out
      (Effects|Components|ParticleSystem|SetIntParameter _returnvalue "SwingSectionIndex" -1))))

(event EventSwingRight
  (bind _returnvalue (CallFunction|GetSwingVFXComponent))
  (|CallParentFunction)
  (bind _section_index (Utilities|SetupSwingMontageSectionIndex (Variables|Actor|GetInstigator))
    (:Out
      (Effects|Components|ParticleSystem|SetIntParameter _returnvalue "SwingSectionIndex" _section_index))
    (:Cast%20Failed%20Out
      (Effects|Components|ParticleSystem|SetIntParameter _returnvalue "SwingSectionIndex" -1))))

(event EventBeginPlay)

(event EventTick
  (params (DeltaSeconds :type Float%20%28single-precision%29)))

(fn ConstructionScript
  (|CallParentFunction))
