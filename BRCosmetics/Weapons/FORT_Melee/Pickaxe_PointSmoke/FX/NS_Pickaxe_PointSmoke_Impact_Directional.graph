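# Niagara directional impact effect for the PointSmoke pickaxe, combining split skull sprites, flung elements, streaks and layered flashes, among other emitters.
# The swing section sets skull mirroring and motion, while a system direction vector aims elements and streaks.
# User.Variant switches six color styles, with front-end modifiers and time-of-day tinting adjusting the look.
# Visibility tags and per-layer particle attributes let single emitters feed several renderers, with flash counts scaled by quality level.
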NiagaraSystem /BRCosmetics/Weapons/FORT_Melee/Pickaxe_PointSmoke/FX/NS_Pickaxe_PointSmoke_Impact_Directional
dependencies {
 usedDataInterfaces [
  &NiagaraDataInterfaceArrayInt32
  &NiagaraDataInterfaceCurve
  &NiagaraDataInterfaceCamera
  &NiagaraDataInterfaceColorCurve
  &NiagaraDataInterfaceArrayFloat
  &NiagaraDataInterfaceArrayColor
  &NiagaraDataInterfaceSpriteRendererInfo
  &NiagaraDataInterfaceVector2DCurve
  &NiagaraDataInterfaceArrayFloat3
 ]
 usedDynamicInputs [
  &UniformRangedVector2D
  /Niagara/DynamicInputs/Math/Clamp_Vector
  &SelectVectorFromArray
  /Game/Effects/Niagara/Modules/Utility/FN_DirectionVectorByPickaxeType
  &SetBoolByIntComparison
 ]
 usedModules [
  &self:SystemSpawnScript.SystemScriptSource.SystemScriptGraph.NiagaraNodeAssignment_2.SetVariables_#1
  &self:SystemSpawnScript.SystemScriptSource.SystemScriptGraph.NiagaraNodeAssignment_10.SetVariables_#2
  &self:SystemSpawnScript.SystemScriptSource.SystemScriptGraph.NiagaraNodeAssignment_1.SetVariables_#3
  &self:SystemSpawnScript.SystemScriptSource.SystemScriptGraph.NiagaraNodeAssignment_0.SetVariables_#4
  &self:SystemSpawnScript.SystemScriptSource.SystemScriptGraph.NiagaraNodeAssignment_11.SetVariables_#5
  &SystemState
 ]
}
emitters {
 CutSkulls001 {
  emitterUpdateScript [
   &EmitterState()
   &SpawnBurst_Instantaneous(Spawn Count=2, Use Spawn Probability=false, UseLoopCountLimit=false)
  ]
  particleSpawnScript [
   &PartitionParticles()
   /SetVariables(Particles.VisibilityTag=&Add_Integer(
     A=&Multiply_Int(A=$Particles.PartitionParticles.Partition, B=4)
     B=&SelectIntFromArray(Direct Array Index=$User.SwingSectionIndex, Int Selection Array=DataInterface() {propertyValues {IntData [1,2,3,4]}})
    )) {isSetParameters=true}
   /SetVariables(Particles.VisibilityTag=&SelectIntFromArray(Direct Array Index=$User.SwingSectionIndex, Int Selection Array=DataInterface() {propertyValues {IntData [1,2,3,4]}})) {enabled=false, isSetParameters=true}
   &InitializeParticle(
    Lifetime=0.4
    Sprite Rotation Mode=::NewEnumerator1
    Sprite Size Mode=::NewEnumerator3
    Sprite UV Mode=::NewEnumerator4
    Sprite UV Scale:Vector2f=&FN_FrontEndVector2Modifier(
     Front End:Vector2f=&MergeHelper_Vector2Switch(0=<-1 1>, 1=<1 1>, 2=<1 1>, 3=<-1 1>, 4=?, Index=$User.SwingSectionIndex, Num=::NewEnumerator2)
     Gameplay:Vector2f=&MergeHelper_Vector2Switch(0=<1 1>, 1=<-1 1>, 2=<1 1>, 3=<1 1>, 4=?, Index=$User.SwingSectionIndex, Num=::NewEnumerator2)
    )
    Uniform Sprite Size=&Multiply_Float(A=45, B=&FN_FrontEndFloatModifier(Gameplay=0.8, Front End=0.4))
    UsePositionOffset=false
    Write Material Random=true
   ) {module=InitializeParticle001}
   &AddVelocity(
    Velocity=&FN_FrontEndVector3Modifier(
     Gameplay=&MergeHelper_VectorSwitch(0=<0 -0.7 -1>, 1=<0 0.7 -1>, 2=<0 1 0.3>, 3=<0 1 0.7>, Index=$User.SwingSectionIndex, Num=::NewEnumerator2)
     Front End=&MergeHelper_VectorSwitch(0=<0 -0.7 -1>, 1=<0 0.7 -1>, 2=<0 1 0.3>, 3=<0 1 0.7>, Index=$User.SwingSectionIndex, Num=::NewEnumerator2)
    )
    Velocity Speed Scale=&Multiply_Float(
     A=&SelectFloatFromArray(Direct Array Index=$Particles.PartitionParticles.Partition, Float Selection Array=DataInterface() {propertyValues {FloatData [50,-50]}})
     B=&FN_FrontEndFloatModifier(Gameplay=1, Front End=$System.FrontendScale)
    )
   )
   &AddVelocity(
    Rotation Coordinate Space=::World
    Velocity=&FN_FrontEndVector3Modifier(
     Gameplay=<0 0 0>
     Front End=&MergeHelper_VectorSwitch(0=<0 -0.7 -1>, 1=<0 0.7 -1>, 2=<0 1 0.3>, 3=<0 1 0.7>, Index=$User.SwingSectionIndex, Num=::NewEnumerator2)
    )
    Velocity Speed Scale=&Multiply_Float(
     A=&SelectFloatFromArray(Direct Array Index=$Particles.PartitionParticles.Partition, Float Selection Array=DataInterface() {propertyValues {FloatData [50,-50]}})
     B=&FN_FrontEndFloatModifier(Gameplay=1, Front End=$System.FrontendScale)
    )
   ) {enabled=false, module=AddVelocity001}
   &CameraOffset(Camera Offset Amount=&FN_FrontEndFloatModifier(Gameplay=25, Front End=10))
  ]
  particleUpdateScript [
   &ParticleState()
   &SolveForcesAndVelocity(Clamp Velocity=false, Limit Acceleration=false)
   &ScaleColor(
    Scale Alpha=&Multiply_Float(
     A=&FloatFromCurve(
      FloatCurve=DataInterface() {propertyValues {Curve {keys [{arriveTangent=-1.9375, interpMode=RCIM_Cubic, leaveTangent=-1.9375, tangentMode=RCTM_User, time=0, value=1}, <at 1 0>]}}}
      Scale Curve=&FNTimeOfDayScalar(Night Scale=1, Morning Scale=1, Evening Scale=1)
     )
     B=&MakeFloatFromLinearColor(LinearColor=$System.TODColorMultiplier, Channel:ENiagaraLinearColor_Channels=::NewEnumerator3)
    )
    Scale RGB=&Multiply_Vector(
     A=&MakeVectorFromLinearColorRGB(Color=&MergeHelper_ColorSwitch(
       0=&ColorFromCurve(ColorCurve=DataInterface() {propertyValues {
          AlphaCurve {keys []}
          BlueCurve {keys [<at 0 1 RCIM_Cubic>,<at 1 0.1476 RCIM_Cubic>]}
          GreenCurve {keys [<at 0 1 RCIM_Cubic>,<at 1 0.1476 RCIM_Cubic>]}
          RedCurve {keys [<at 0 1 RCIM_Cubic>,<at 1 0.1476 RCIM_Cubic>]}
         }})
       1=&ColorFromCurve(ColorCurve=DataInterface() {propertyValues {
          AlphaCurve {keys [<at 0 1 RCIM_Cubic>,<at 1 0 RCIM_Cubic>]}
          BlueCurve {keys [<at 0 0.076 RCIM_Cubic>,<at 1 0.0072 RCIM_Cubic>]}
          GreenCurve {keys [<at 0 1 RCIM_Cubic>,<at 1 0.0615 RCIM_Cubic>]}
          RedCurve {keys [<at 0 0.8922 RCIM_Cubic>,<at 1 0.0677 RCIM_Cubic>]}
         }})
       2=&ColorFromCurve(ColorCurve=DataInterface() {propertyValues {
          AlphaCurve {keys [<at 0 1 RCIM_Cubic>,<at 1 0 RCIM_Cubic>]}
          BlueCurve {keys [<at 0 0.0731 RCIM_Cubic>,<at 1 0.0118 RCIM_Cubic>]}
          GreenCurve {keys [<at 0 0.5366 RCIM_Cubic>,<at 1 0.0866 RCIM_Cubic>]}
          RedCurve {keys [<at 0 1 RCIM_Cubic>,<at 1 0.1615 RCIM_Cubic>]}
         }})
       3=&ColorFromCurve(ColorCurve=DataInterface() {propertyValues {
          AlphaCurve {keys [<at 0 1 RCIM_Cubic>,<at 1 0 RCIM_Cubic>]}
          BlueCurve {keys [<at 0 1 RCIM_Cubic>,<at 1 0.151 RCIM_Cubic>]}
          GreenCurve {keys [<at 0 0.9153 RCIM_Cubic>,<at 1 0.1382 RCIM_Cubic>]}
          RedCurve {keys [<at 0 0.67 RCIM_Cubic>,<at 1 0.1011 RCIM_Cubic>]}
         }})
       4=&ColorFromCurve(ColorCurve=DataInterface() {propertyValues {
          AlphaCurve {keys [<at 0 1 RCIM_Cubic>,<at 1 0 RCIM_Cubic>]}
          BlueCurve {keys [<at 0 2 RCIM_Cubic>,<at 0.6693 0.2292 RCIM_Cubic>]}
          GreenCurve {keys [<at 0 1.2162 RCIM_Cubic>,<at 0.6693 0.0757 RCIM_Cubic>]}
          RedCurve {keys [<at 0 0.1188 RCIM_Cubic>,<at 0.6693 0.2164 RCIM_Cubic>]}
         }})
       5=&ColorFromCurve(ColorCurve=DataInterface() {propertyValues {
          AlphaCurve {keys [<at 0 1 RCIM_Cubic>,<at 1 0 RCIM_Cubic>]}
          BlueCurve {keys [<at 0.0756 1 RCIM_Cubic>,<at 0.9568 0.1302 RCIM_Cubic>]}
          GreenCurve {keys [<at 0.0756 0.514 RCIM_Cubic>,<at 0.9568 0.0401 RCIM_Cubic>]}
          RedCurve {keys [<at 0.0756 0.2405 RCIM_Cubic>,<at 0.9568 0.0604 RCIM_Cubic>]}
         }})
       Index=&MakeIntFromFloat(Float=$User.Variant)
       Num=::NewEnumerator4
      ))
     B=&MakeVectorFromLinearColorRGB(Color=$System.TODColorMultiplier)
    )
    ScaleA=true
    ScaleRGB=true
   )
  ]
  rendererClasses [&NiagaraSpriteRendererProperties]
  renderers [
   {rendererClass=&NiagaraSpriteRendererProperties}
   {rendererClass=&NiagaraSpriteRendererProperties, rendererIndex=1}
   {rendererClass=&NiagaraSpriteRendererProperties, rendererIndex=2}
   {rendererClass=&NiagaraSpriteRendererProperties, rendererIndex=3}
   {rendererClass=&NiagaraSpriteRendererProperties, rendererIndex=4}
   {rendererClass=&NiagaraSpriteRendererProperties, rendererIndex=5}
   {rendererClass=&NiagaraSpriteRendererProperties, rendererIndex=6}
   {rendererClass=&NiagaraSpriteRendererProperties, rendererIndex=7}
  ]
 }
 Directional_Elements {
  emitterUpdateScript [
   &EmitterState(
    Enable Distance Culling=true
    Enable Visibility Culling=true
    MaxDistance=2500
    Module.bApplySpawnCountScaleByDistance=true
    Module.bUseMaxDistance=true
    Module.bUseMinDistance=false
    Scalability Mode=::NewEnumerator1
    Scale Spawn Count=true
   )
   &SpawnBurst_Instantaneous(
    Spawn Count=&Multiply_Int(
     A=&RandomRangeInt(Minimum=2, Maximum=4)
     B=&MakeCustomIntFromBool(Bool=&Boolean_OrOperation(Bool1=$System.IsHolo, Bool2=$System.IsIce), True INT=1, False INT=2)
    )
    Use Spawn Probability=true
    UseLoopCountLimit=false
   )
  ]
  particleSpawnScript [
   &InitializeParticle(
    Color=&MergeHelper_ColorSwitch(
     0=<rgba 0 0 0 0>
     1=&ColorFromCurve(
      ColorCurve=DataInterface() {propertyValues {
        AlphaCurve {keys [<at 0 1 RCIM_Cubic>]}
        BlueCurve {keys [<at 0.4274 0.9047 RCIM_Cubic>,<at 0.4699 0.6751 RCIM_Cubic>]}
        GreenCurve {keys [<at 0.4274 0.8796 RCIM_Cubic>,<at 0.4699 0.6696 RCIM_Cubic>]}
        RedCurve {keys [<at 0.4274 1 RCIM_Cubic>,<at 0.4699 1 RCIM_Cubic>]}
       }}
      CurveIndex=&RandomRangeFloat(Minimum=0, Maximum=1)
     )
     2=&ColorFromCurve(
      ColorCurve=DataInterface() {propertyValues {
        AlphaCurve {keys [<at 0 1 RCIM_Cubic>]}
        BlueCurve {keys [<at 0.4274 0 RCIM_Cubic>,<at 0.4699 0 RCIM_Cubic>]}
        GreenCurve {keys [<at 0.4274 0.6232 RCIM_Cubic>,<at 0.4699 0.3648 RCIM_Cubic>]}
        RedCurve {keys [<at 0.4274 1 RCIM_Cubic>,<at 0.4699 1 RCIM_Cubic>]}
       }}
      CurveIndex=&RandomRangeFloat(Minimum=0, Maximum=1)
     )
     3=&ColorFromCurve(
      ColorCurve=DataInterface() {propertyValues {
        AlphaCurve {keys [<at 0 1 RCIM_Cubic>]}
        BlueCurve {keys [<at 0.4274 1 RCIM_Cubic>,<at 0.4699 1 RCIM_Cubic>]}
        GreenCurve {keys [<at 0.4274 0.9469 RCIM_Cubic>,<at 0.4699 0.6167 RCIM_Cubic>]}
        RedCurve {keys [<at 0.4274 0.5208 RCIM_Cubic>,<at 0.4699 0.4375 RCIM_Cubic>]}
       }}
      CurveIndex=&RandomRangeFloat(Minimum=0, Maximum=1)
     )
     4=&ColorFromCurve(
      ColorCurve=DataInterface() {propertyValues {
        AlphaCurve {keys [<at 0 1 RCIM_Cubic>]}
        BlueCurve {keys [<at 0.4274 1 RCIM_Cubic>,<at 0.4699 1 RCIM_Cubic>]}
        GreenCurve {keys [<at 0.4274 0.7649 RCIM_Cubic>,<at 0.4699 0.3664 RCIM_Cubic>]}
        RedCurve {keys [<at 0.4274 0.3886 RCIM_Cubic>,<at 0.4699 0.7831 RCIM_Cubic>]}
       }}
      CurveIndex=&RandomRangeFloat(Minimum=0, Maximum=1)
     )
     5=&ColorFromCurve(
      ColorCurve=DataInterface() {propertyValues {
        AlphaCurve {keys [<at 0 1 RCIM_Cubic>]}
        BlueCurve {keys [<at 0.4274 1 RCIM_Cubic>,<at 0.4699 1 RCIM_Cubic>]}
        GreenCurve {keys [<at 0.4274 0 RCIM_Cubic>,<at 0.4699 0.282 RCIM_Cubic>]}
        RedCurve {keys [<at 0.4274 0.6698 RCIM_Cubic>,<at 0.4699 0 RCIM_Cubic>]}
       }}
      CurveIndex=&RandomRangeFloat(Minimum=0, Maximum=1)
     )
     Index=&MakeIntFromFloat(Float=$User.Variant, Conversion Mode=::NewEnumerator1)
     Num=::NewEnumerator4
    )
    Lifetime=&RandomRangeFloat(Minimum=0.3, Maximum=0.6)
    Mesh Scale Mode=::NewEnumerator1
    Mesh Uniform Scale Max=0.4
    Mesh Uniform Scale Min=0.3
    Sprite Rotation Angle=&MergeHelper_FloatSwitch(
     0=0
     1=&RandomRangeFloat(Minimum=0, Maximum=1)
     2=0
     3=0
     4=0
     5=&RandomRangeFloat(Minimum=0, Maximum=1)
     Index=&MakeIntFromFloat(Float=$User.Variant, Conversion Mode=::NewEnumerator1)
     Num=::NewEnumerator4
    )
    Sprite Rotation Mode=::NewEnumerator1
    Sprite Size Mode=::NewEnumerator3
    Uniform Sprite Size=&Multiply_Float(
     A=&Multiply_Float(
      A=&RandomRangeFloat(Minimum=10, Maximum=14)
      B=&FN_FrontEndFloatModifier(Gameplay=1, Front End=$System.FrontendScale)
     )
     B=&SelectFloatFromArray(
      Direct Array Index=&MakeIntFromFloat(Float=$User.Variant, Conversion Mode=::NewEnumerator1)
      Float Selection Array=DataInterface() {propertyValues {FloatData [0,1,1,0.75,1.5,1.2]}}
     )
    )
    UsePositionOffset=true
    Write Material Random=true
   )
   &ShapeLocation(
    MaskSpawnGroup=false
    Override Seed=false
    Sphere Radius=&Multiply_Float(
     A=&Multiply_Float(A=20, B=&FN_FrontEndFloatModifier(Gameplay=0.5, Front End=$System.FrontendScale))
     B=&SelectFloatFromArray(
      Direct Array Index=&MakeIntFromFloat(Float=$User.Variant, Conversion Mode=::NewEnumerator1)
      Float Selection Array=DataInterface() {propertyValues {FloatData [1,1,1,1,1,2]}}
     )
    )
    Transform Order:ENiagara_TransformOrder=::NewEnumerator1
    UseOwnerScale=true
   ) {module=ShapeLocation001}
   &CameraOffset(Camera Offset Amount=&FN_FrontEndFloatModifier(Gameplay=15, Front End=0)) {module=CameraOffset001}
   &AddVelocity(
    Module.Use Falloff Distance=false
    Override Seed=false
    Velocity Mode=::NewEnumerator1
    Velocity Speed=&FN_FrontEndFloatModifier(Gameplay=0, Front End=&RandomRangeFloat(Minimum=100, Maximum=200))
   ) {module=AddVelocity001}
   &AddVelocity(
    Cone Angle=&RandomRangeFloat(Minimum=20, Maximum=25)
    Cone Axis=&Multiply_VectorByFloat(Vector=$System.DirectionVector, Float=1)
    Module.UseConeFalloff=false
    Module.UseVelDistribution=false
    Override Seed=false
    Velocity Mode=::NewEnumerator2
    Velocity Speed=&Multiply_Float(
     A=&RandomRangeFloat(Minimum=75, Maximum=1000)
     B=&FN_FrontEndFloatModifier(Gameplay=1, Front End=0)
    )
   )
   &InitialMeshOrientation(
    Mesh Orientation Mode=::NewEnumerator2
    Orientation Vector=$System.DirectionVector
    Rotation=&RandomRangeVector(Minimum=<0 0 0>, Maximum=<0 0 1>)
    Use Rotation Vector=true
   )
  ]
  particleUpdateScript [
   &ParticleState()
   &ScaleMeshSize(Scale Factor=&VectorFromFloat(Value=&FloatFromCurve(FloatCurve=DataInterface() {propertyValues {Curve {keys [<at 0 1 RCIM_Cubic>,<at 1 0 RCIM_Cubic>]}}})))
   &UpdateMeshOrientation(
    Facing Direction=<0 0 -1>
    Facing Interpolation=false
    Orientation Method=::NewEnumerator1
    Override Quat=false
    TransformFacing=false
    TransformUp=false
   ) {enabled=false}
   &AccelerationForce(Acceleration=&Multiply_VectorByFloat(
     Vector=&MaskVectorBySpawnGroup(Masked Vector=<0 0 -150>, Passthrough Vector=<0 0 1>)
     Float=&FN_FrontEndFloatModifier(Gameplay=1, Front End=$System.FrontendScale)
    ))
   &Drag(
    Drag=&Multiply_Float(
     A=&MaskFloatBySpawnGroup(Masked Float=6, Passthrough Float=8)
     B=&MergeHelper_FloatSwitch(
      0=1
      1=1
      2=1
      3=1
      4=1
      5=0.4
      Index=&MakeIntFromFloat(Float=$User.Variant, Conversion Mode=::NewEnumerator1)
      Num=::NewEnumerator4
     )
    )
    Use Linear Drag=true
    Use Rotational Drag=true
   )
   &CurlNoiseForce(Bias Noise=false, Noise Strength=&RandomRangeFloat(Minimum=80, Maximum=120), Pan Noise=false)
   &SolveForcesAndVelocity(Clamp Velocity=false, Limit Acceleration=false)
   &ScaleSpriteSize(
    Uniform Curve Scale=&FN_FrontEndFloatModifier(Gameplay=1, Front End=$System.FrontendScale)
    Uniform Curve Sprite Scale=DataInterface() {propertyValues {Curve {keys [
        {arriveTangent=9.0725, interpMode=RCIM_Cubic, leaveTangent=3.7536, tangentMode=RCTM_Break, time=0, value=0}
        <at 0.5 1 RCIM_Cubic>
        {arriveTangent=-0.0405, interpMode=RCIM_Cubic, leaveTangent=-0.0405, tangentMode=RCTM_User, time=1, value=0}
       ]}}}
   )
   &SubUVAnimation(
    Play Rate:NiagaraFloat=&RandomRangeFloat(Minimum=0.5, Maximum=1)
    Playback Scale Curve=DataInterface() {propertyValues {Curve {keys [
        {interpMode=RCIM_Cubic, tangentMode=RCTM_User, time=0.25, value=1}
        {arriveTangent=-1.2656, interpMode=RCIM_Cubic, leaveTangent=-1.2656, tangentMode=RCTM_User, time=1, value=0}
       ]}}}
    Random Start Frame=true
    "Sprite Renderer "=DataInterface() {propertyValues {SpriteRenderer=&self:Directional_Elements_0.NiagaraSpriteRendererProperties_4}}
    SubUV Animation Mode:ENiagara_SubUVLookupModeV2=::NewEnumerator3
    UseEndFrame=false
    UsePlayRateScale=false
    UseStartFrame=false
   )
   &ScaleColor(
    Scale Alpha=&MakeFloatFromLinearColor(LinearColor=$System.TODColorMultiplier, Channel:ENiagaraLinearColor_Channels=::NewEnumerator3)
    Scale RGB=&Multiply_Vector(
     A=&MakeVectorFromLinearColorRGB(Color=&ColorFromCurve(ColorCurve=DataInterface() {propertyValues {
         AlphaCurve {keys [<at 0 1 RCIM_Cubic>,<at 1 0 RCIM_Cubic>]}
         BlueCurve {keys [<at 0.0038 3 RCIM_Cubic>, {interpMode=RCIM_Cubic, leaveTangent=-2.2104, time=0.3007, value=1}]}
         GreenCurve {keys [<at 0.0038 3 RCIM_Cubic>, {interpMode=RCIM_Cubic, leaveTangent=-2.2104, time=0.3007, value=1}]}
         RedCurve {keys [<at 0.0038 3 RCIM_Cubic>, {interpMode=RCIM_Cubic, leaveTangent=-2.2104, time=0.3007, value=1}]}
        }}))
     B=&MakeVectorFromLinearColorRGB(Color=$System.TODColorMultiplier)
    )
    ScaleA=true
    ScaleRGB=true
   )
  ]
  rendererClasses [&NiagaraSpriteRendererProperties, &NiagaraMeshRendererProperties]
  renderers [
   {rendererClass=&NiagaraSpriteRendererProperties}
   {rendererClass=&NiagaraSpriteRendererProperties, rendererIndex=1}
   {rendererClass=&NiagaraSpriteRendererProperties, rendererIndex=2}
   {rendererClass=&NiagaraSpriteRendererProperties, rendererIndex=3}
   {rendererClass=&NiagaraMeshRendererProperties, rendererIndex=4}
  ]
 }
 Directional_Streaks001 {
  emitterUpdateScript [
   &EmitterState()
   &SpawnBurst_Instantaneous(
    Spawn Probability=&FN_FrontEndFloatModifier(Gameplay=1, Front End=1)
    Use Spawn Probability=true
    UseLoopCountLimit=false
   )
   &SpawnBurst_Instantaneous(
    Spawn Group=1
    Spawn Probability=&FN_FrontEndFloatModifier(Gameplay=1, Front End=1)
    Use Spawn Probability=true
    UseLoopCountLimit=false
   ) {module=SpawnBurst_Instantaneous001}
  ]
  particleSpawnScript [
   /SetVariables(Particles.VisibilityTag=$Emitter.SpawnGroup) {isSetParameters=true}
   &InitializeParticle(
    Color=&MultiplyLinearColor(
     Linear Color=$System.TODColorMultiplier
     Scale Factor=&MakeVector4_FromLinearColor(LinearColor=&MergeHelper_ColorSwitch(
       0=&RandomRangeLinearColor(Minimum=<rgba 1 1 1 0.3>, Maximum=<rgba 1 1 1 0.5>)
       1=&RandomRangeLinearColor(Minimum=<rgba 0.1577 0.1892 0 1>, Maximum=<rgba 0.1107 0.1476 0 1>)
       2=&RandomRangeLinearColor(Minimum=<rgba 0.2 0.1382 0.0145 1>, Maximum=<rgba 0.1507 0.0869 0.0232 1>)
       3=&RandomRangeLinearColor(Minimum=<rgba 0 0.15 0.2 1>, Maximum=<rgba 0 0.1667 0.25 1>)
       4=&RandomRangeLinearColor(Minimum=<rgba 0.0909 0.1474 0.3826 1>, Maximum=<rgba 0.0496 0.3 0.2669 1>)
       5=&RandomRangeLinearColor(Minimum=<rgba 0.0909 0.2854 0.3826 1>, Maximum=<rgba 0.0496 0.1748 0.3 1>)
       Index=&MakeIntFromFloat(Float=$User.Variant)
       Num=::NewEnumerator4
      ))
    )
    Lifetime Max=0.2
    Lifetime Min=0.15
    Lifetime Mode=::NewEnumerator1
    Sprite Size=&FN_FrontEndVector2Modifier(
     Front End:Vector2f=&RandomRangeVector2D(Minimum=<4 37.5>, Maximum=<6 50>)
     Gameplay:Vector2f=&RandomRangeVector2D(Minimum=<8 75>, Maximum=<12 100>)
    )
    Sprite Size Mode=::NewEnumerator4
    Sprite UV Mode=::NewEnumerator1
    UsePositionOffset=false
    Write Material Random=true
   ) {module=InitializeParticle001}
   &SubUVAnimation(End Frame=3, SubUV Animation Mode:ENiagara_SubUVLookupMode=::NewEnumerator2)
   &AddVelocity(
    Velocity=$System.DirectionVector
    Velocity Speed Scale=&MergeHelper_FloatSwitch(
     0=&RandomRangeFloat(Minimum=3, Maximum=3)
     1=&RandomRangeFloat(Minimum=150, Maximum=250)
     2=?
     3=?
     Index=$Emitter.SpawnGroup
     Num=::NewEnumerator0
    )
   )
   &CameraOffset(Camera Offset Amount=25)
   /SetVariables(Particles.Size Blurry Streak:Vector2f=&FN_FrontEndVector2Modifier(
     Front End:Vector2f=&RandomRangeVector2D(Minimum=<20 62.5>, Maximum=<30 75>)
     Gameplay:Vector2f=&RandomRangeVector2D(Minimum=<40 125>, Maximum=<60 150>)
    )) {isSetParameters=true}
   /SetVariables(
    Particles.Color Dark Streak:LinearColor=&MergeHelper_ColorSwitch(
     0=&RandomRangeLinearColor(Minimum=<rgba 0 0 0 1>, Maximum=<rgba 0.02 0.02 0.02 1>)
     1=&RandomRangeLinearColor(Minimum=<rgba 0.087 0.087 0.0054 1.5>, Maximum=<rgba 0.0268 0.029 0.0032 1.5>)
     2=&RandomRangeLinearColor(Minimum=<rgba 0.1391 0.0986 0.0417 1.5>, Maximum=<rgba 0.0986 0.0489 0.0134 1.5>)
     3=&RandomRangeLinearColor(Minimum=<rgba 0.0105 0.1287 0.1681 1.5>, Maximum=<rgba 0.0095 0.0547 0.087 1.5>)
     4=&RandomRangeLinearColor(Minimum=<rgba 0.4232 0.1692 0.6291 1>, Maximum=<rgba 0.7021 0.1822 0.806 0.9>)
     5=&RandomRangeLinearColor(Minimum=<rgba 0.0888 0.0027 0.2609 0.8>, Maximum=<rgba 0.0651 0.002 0.1913 0.85>)
     Index=&MakeIntFromFloat(Float=$User.Variant)
     Num=::NewEnumerator4
    )
    Particles.Size Dark Streak:Vector2f=&FN_FrontEndVector2Modifier(
     Front End:Vector2f=&RandomRangeVector2D(Minimum=<8 62.5>, Maximum=<9 67.5>)
     Gameplay:Vector2f=&RandomRangeVector2D(Minimum=<15 120>, Maximum=<16 130>)
    )
   ) {isSetParameters=true}
  ]
  particleUpdateScript [
   &ParticleState()
   &SolveForcesAndVelocity(Clamp Velocity=false, Limit Acceleration=false)
   &ScaleSpriteSize(Scale Factor=&Vector2DFromCurve(
     Vector2Curve=DataInterface() {propertyValues {XCurve {keys [<at 0 1 RCIM_Cubic>,<at 1 0 RCIM_Cubic>]}, YCurve {keys [<at 0 0.5 RCIM_Cubic>,<at 1 2 RCIM_Cubic>]}}}
     ScaleCurve=&Multiply_Vector2D_ByFloat(Float=1, Vector2D=<1.5 1>)
    ))
   &ScaleColor(
    Scale Alpha=&FloatFromCurve(
     FloatCurve=DataInterface() {propertyValues {Curve {keys [<at 0 1 RCIM_Cubic>,<at 1 0>]}}}
     Scale Curve=&FNTimeOfDayScalar(Night Scale=1, Morning Scale=1, Evening Scale=1)
    )
    Scale RGB=&VectorFromFloat(Value=&FloatFromCurve(FloatCurve=DataInterface() {propertyValues {Curve {keys [<at 0 10 RCIM_Cubic>,<at 1 1 RCIM_Cubic>]}}}))
    ScaleA=true
    ScaleRGB=true
   )
   /SetVariables(
    Particles.Update Color Dark Streak:LinearColor=&MultiplyLinearColor(
     Linear Color=$System.TODColorMultiplier
     Scale Factor=&MakeVector4_FromLinearColor(LinearColor=&MultiplyLinearColor(
       Linear Color=$Particles.Color Dark Streak
       Scale Factor=&MakeVector4(
        X=1
        Y=1
        Z=1
        W=&FloatFromCurve(FloatCurve=DataInterface() {propertyValues {Curve {keys [<at 0 1 RCIM_Cubic>,<at 1 0>]}}})
       )
      ))
    )
    Particles.Update Size Dark Streak:Vector2f=&Multiply_Vector2D(
     A=$Particles.Size Dark Streak
     B=&Vector2DFromCurve(
      Vector2Curve=DataInterface() {propertyValues {XCurve {keys [<at 0 1 RCIM_Cubic>,<at 1 0 RCIM_Cubic>]}, YCurve {keys [<at 0 0.5 RCIM_Cubic>,<at 1 2 RCIM_Cubic>]}}}
      ScaleCurve=&Multiply_Vector2D_ByFloat(Float=1, Vector2D=<1.5 1>)
     )
    )
   ) {isSetParameters=true}
   &FN_Sprite_Attributes(
    Camera Offset=30
    Color=&MakeLinearColorFromVectorAndFloat(
     "Vector (RGB)"=&MakeVectorFromLinearColorRGB(Color=&MultiplyLinearColor(
       Linear Color=&MergeHelper_ColorSwitch(
        0=<rgba 10 10 10 1>
        1=<rgba 13.75 15 0 1>
        2=<rgba 5 2.3623 0.4783 1>
        3=<rgba 0.4783 5 5 1>
        4=<rgba 4.2464 0.4783 5 1>
        5=<rgba 2.1362 0.4783 5 1>
        Index=&MakeIntFromFloat(Float=$User.Variant)
        Num=::NewEnumerator4
       )
       Scale Factor=&MakeVector4_FromLinearColor(LinearColor=$System.TODColorMultiplier)
      ))
     "Float (Alpha)"=&Multiply_Float(
      A=&MakeFloatFromLinearColor(LinearColor=$System.TODColorMultiplier, Channel:ENiagaraLinearColor_Channels=::NewEnumerator3)
      B=&FloatFromCurve(
       FloatCurve=DataInterface() {propertyValues {Curve {keys [{interpMode=RCIM_Cubic, tangentMode=RCTM_User, time=0, value=1}, {interpMode=RCIM_Cubic, tangentMode=RCTM_User, time=1, value=0}]}}}
       Scale Curve=0.75
      )
     )
    )
    Create Camera Offset Attribute=true
    Create Color Attribute=true
    Create SpriteSize Attribute=true
    SpriteSize=&Multiply_Vector2D(
     A=&Multiply_Vector2D_ByFloat(Float=&RandomRangeFloat(Minimum=0.95, Maximum=1.05), Vector2D=<15 300>)
     B=&Vector2DFromCurve(
      Vector2Curve=DataInterface() {propertyValues {
        XCurve {keys [{tangentMode=RCTM_User, time=0, value=0}, <at 0.2 1.075>, {tangentMode=RCTM_User, time=1, value=-0.1}]}
        YCurve {keys [{tangentMode=RCTM_User, time=0, value=0}, <at 0.1 1.025>, {tangentMode=RCTM_User, time=1, value=0.8}]}
       }}
      ScaleCurve=<1 1>
     )
    )
    Write to Module Namespace=true
   )
   /SetVariables() {isSetParameters=true}
  ]
  rendererClasses [&NiagaraSpriteRendererProperties]
  renderers [
   {rendererClass=&NiagaraSpriteRendererProperties}
   {rendererClass=&NiagaraSpriteRendererProperties, rendererIndex=1}
   {rendererClass=&NiagaraSpriteRendererProperties, rendererIndex=2}
   {rendererClass=&NiagaraSpriteRendererProperties, rendererIndex=3}
  ]
 }
 Flash_Elements {
  emitterUpdateScript [
   &EmitterState()
   &SpawnBurst_Instantaneous(Spawn Group=1, Use Spawn Probability=false, UseLoopCountLimit=false) {module=SpawnBurst_Instantaneous001}
   &SpawnBurst_Instantaneous(
    Spawn Count=&RandomRangeInt(
     Minimum=&SelectIntFromArray(Direct Array Index=$Engine.QualityLevel, Int Selection Array=DataInterface() {propertyValues {IntData [5,6,8,10,10]}})
     Maximum=&SelectIntFromArray(Direct Array Index=$Engine.QualityLevel, Int Selection Array=DataInterface() {propertyValues {IntData [8,9,10,15,15]}})
    )
    Use Spawn Probability=false
    UseLoopCountLimit=false
   )
   &SpawnBurst_Instantaneous(
    Spawn Count=&RandomRangeInt(
     Minimum=&SelectIntFromArray(Direct Array Index=$Engine.QualityLevel, Int Selection Array=DataInterface() {propertyValues {IntData [1,2,2,3,3]}})
     Maximum=&SelectIntFromArray(Direct Array Index=$Engine.QualityLevel, Int Selection Array=DataInterface() {propertyValues {IntData [2,3,4,5,5]}})
    )
    Spawn Group=2
    Use Spawn Probability=false
    UseLoopCountLimit=false
   ) {module=SpawnBurst_Instantaneous002}
  ]
  particleSpawnScript [
   &InitializeParticle(
    Color Mode=::NewEnumerator0
    Lifetime=&MakeCustomFloatFromBool(Bool=$System.IsGalaxy, True Float=0.55, False Float=0.25)
    UsePositionOffset=false
    Write Material Random=true
   )
   &ShapeLocation(
    MaskSpawnGroup=false
    Override Seed=false
    Sphere Radius=&MakeCustomFloatFromBool(Bool=$System.IsGalaxy, True Float=1, False Float=1)
    UseOwnerScale=true
   )
   &SpriteFacingAndAlignment(Sprite Alignment=$Particles.ShapeLocation.ShapeNormal, WriteAlignment=true, WriteFacing=false)
   &FN_PerParticleLOD(Enable SpriteSize Distance Scaling=true, Sprite Size Scale Far=<2.5 2.5>)
   /SetVariables(Particles.SpriteSize=<1 1>, Particles.VisibilityTag=&SpawnGroup()) {isSetParameters=true}
   /SetVariables(
    Particles.Flash Spike Color:LinearColor=&MergeHelper_ColorSwitch(
     0=<rgba 0 0 0 0.8>
     1=<rgba 0.0226 0.0226 0.0028 0.8>
     2=<rgba 0.1 0.0533 0.02 0.8>
     3=<rgba 0.0125 0.0708 0.1 0.8>
     4=<rgba 0.0475 0.077 0.2 3>
     5=<rgba 0.0475 0.077 0.2 3>
     Index=&MakeIntFromFloat(Float=$User.Variant)
     Num=::NewEnumerator4
    )
    Particles.Flash Spike Lifetime:NiagaraFloat=&RandomRangeFloat(Minimum=0.05, Maximum=0.08)
    Particles.Flash Spike Sprite Size:Vector2f=&FN_FrontEndVector2Modifier(
     Front End:Vector2f=&RandomRangeVector2D(Minimum=<1.5 27.5>, Maximum=<5 35>)
     Gameplay:Vector2f=&RandomRangeVector2D(Minimum=<3 55>, Maximum=<10 65>)
    )
   ) {isSetParameters=true}
   /SetVariables(
    Particles.Flash Star Cam Offset=15
    Particles.Flash Star Lifetime=0.15
    Particles.Flash Star Rotation:NiagaraFloat=&RandomRangeFloat(Minimum=0, Maximum=360)
    Particles.Flash Star Sprite Size:Vector2f=&Multiply_Vector2D_ByFloat(
     Float=&Multiply_Float(
      A=&MakeCustomFloatFromBool(Bool=$User.bIsFrontEnd, True Float=0.5, False Float=1)
      B=&RandomRangeFloat(Minimum=120, Maximum=150)
     )
     Vector2D=$Particles.SpriteSize
    )
    Particles.Flash Star Sprite Size Alt:Vector2f=&Multiply_Vector2D_ByFloat(
     Float=&Multiply_Float(
      A=0.75
      B=&Multiply_Float(
       A=&MakeCustomFloatFromBool(Bool=$User.bIsFrontEnd, True Float=0.5, False Float=1)
       B=&RandomRangeFloat(Minimum=120, Maximum=150)
      )
     )
     Vector2D=$Particles.SpriteSize
    )
    Particles.Flash Star SubUV Index:NiagaraFloat=&RandomRangeFloat(Minimum=0, Maximum=3)
    Particles.Flash Star SubUV Index Alt:NiagaraFloat=&RandomRangeFloat(Minimum=0, Maximum=3)
   ) {isSetParameters=true}
   /SetVariables(
    Particles.Flash Flare Color:LinearColor=<rgba 1 1 1 0.7>
    Particles.Flash Flare Lifetime=0.2
    Particles.Flash Flare Sprite Size:Vector2f=&Multiply_Vector2D_ByFloat(
     Float=&Multiply_Float(
      A=&MakeCustomFloatFromBool(Bool=$User.bIsFrontEnd, True Float=0.5, False Float=0.75)
      B=&RandomRangeFloat(Minimum=165, Maximum=175)
     )
     Vector2D=$Particles.SpriteSize
    )
   ) {isSetParameters=true}
   /SetVariables(
    Particles.GalaxySmoke Lifetime:NiagaraFloat=&RandomRangeFloat(Minimum=0.25, Maximum=0.35)
    Particles.GalaxyWave Lifetime=0.5
   ) {isSetParameters=true}
   /SetVariables(Particles.Dark_Glow_CameraOffset=&MakeCustomFloatFromBool(Bool=$User.bIsFrontEndPreview, True Float=1, False Float=50)) {isSetParameters=true}
   &InitialMeshOrientation(
    Mesh Orientation Mode=::NewEnumerator7
    Rotation=&FN_FrontEndVector3Modifier(Gameplay=<0 0.25 0>, Front End=<0 0.25 0>)
    Use Rotation Vector=true
   )
   &AddVelocity(Velocity=&FN_FrontEndVector3Modifier(
     Gameplay=<10 0 0>
     Front End=&Multiply_Vector(A=&ReturnCameraVector(Camera Vectors=::NewEnumerator1, Coordinate Space=::Local), B=<-1 -1 -1>)
    ))
   &ApplyInitialForces()
  ]
  particleUpdateScript [
   &ParticleState()
   /SetVariables(
    Particles.HoloLifetime:NiagaraFloat=&Multiply_Float(A=$Particles.NormalizedAge, B=1)
    Particles.Update GalaxySmoke Lifetime:NiagaraFloat=&Divide_Float(A=$Particles.Age, B=$Particles.Initial.GalaxySmoke Lifetime)
    Particles.Update GalaxyWave Lifetime:NiagaraFloat=&Divide_Float(A=$Particles.Age, B=$Particles.Initial.GalaxyWave Lifetime)
   ) {isSetParameters=true}
   /SetVariables(
    Particles.Update Flash Flare Age:NiagaraFloat=&Divide_Float(A=$Particles.Age, B=$Particles.Initial.Flash Flare Lifetime)
    Particles.Update Flash Spike Age:NiagaraFloat=&Divide_Float(A=$Particles.Age, B=$Particles.Initial.Flash Spike Lifetime)
    Particles.Update Flash Star Age:NiagaraFloat=&Divide_Float(A=$Particles.Age, B=$Particles.Initial.Flash Star Lifetime)
   ) {isSetParameters=true}
   /SetVariables(
    Particles.Update Flash Spike Color:LinearColor=&MakeLinearColorFromVectorAndFloat(
     "Vector (RGB)"=&MakeVectorFromLinearColorRGB(Color=$Particles.Initial.Flash Spike Color)
     "Float (Alpha)"=&FloatFromCurve(FloatCurve=DataInterface() {propertyValues {Curve {keys [<at 0.7 1>,<at 1 0>]}}}, CurveIndex=$Particles.Update Flash Spike Age)
    )
    Particles.Update Flash Spike Sprite Size:Vector2f=&Multiply_Vector2D_ByFloat(
     Float=&FloatFromCurve(FloatCurve=DataInterface() {propertyValues {Curve {keys [<at 0 0.6>,<at 1 2>]}}}, CurveIndex=$Particles.Update Flash Spike Age)
     Vector2D=$Particles.Initial.Flash Spike Sprite Size
    )
    Particles.Update Flash Star Color:LinearColor=&MergeHelper_ColorSwitch(
     0=&ColorFromCurve(
      ColorCurve=DataInterface() {propertyValues {
        AlphaCurve {keys [<at 0 2>,<at 0.8992 0>]}
        BlueCurve {keys [<at 0.1978 1>]}
        GreenCurve {keys [<at 0.1978 1>]}
        RedCurve {keys [<at 0.1978 1>]}
       }}
      CurveIndex=$Particles.Update Flash Star Age
     )
     1=&ColorFromCurve(
      ColorCurve=DataInterface() {propertyValues {
        AlphaCurve {keys [<at 0 2>,<at 0.8992 0>]}
        BlueCurve {keys [<at 0.1978 0.0133>,<at 0.4068 0.0128>,<at 0.6127 0.0068>]}
        GreenCurve {keys [<at 0.1978 0.1275>,<at 0.4068 0.0638>,<at 0.6127 0.0464>]}
        RedCurve {keys [<at 0.1978 0.1085>,<at 0.4068 0.0595>,<at 0.6127 0.0464>]}
       }}
      CurveIndex=$Particles.Update Flash Star Age
     )
     2=&ColorFromCurve(
      ColorCurve=DataInterface() {propertyValues {
        AlphaCurve {keys [<at 0 2>,<at 0.8992 0>]}
        BlueCurve {keys [<at 0.1978 0.0133>,<at 0.4068 0.0128>,<at 0.6127 0.0068>]}
        GreenCurve {keys [<at 0.1978 0.0895>,<at 0.4068 0.0425>,<at 0.6127 0.0266>]}
        RedCurve {keys [<at 0.1978 0.1275>,<at 0.4068 0.0638>,<at 0.6127 0.0464>]}
       }}
      CurveIndex=$Particles.Update Flash Star Age
     )
     3=&ColorFromCurve(
      ColorCurve=DataInterface() {propertyValues {
        AlphaCurve {keys [<at 0 2>,<at 0.8992 0>]}
        BlueCurve {keys [<at 0.1978 0.1275>,<at 0.4068 0.0638>,<at 0.6127 0.0464>]}
        GreenCurve {keys [<at 0.1978 0.1085>,<at 0.4068 0.051>,<at 0.6127 0.0332>]}
        RedCurve {keys [<at 0.1978 0.0133>,<at 0.4068 0.0128>,<at 0.6127 0.0068>]}
       }}
      CurveIndex=$Particles.Update Flash Star Age
     )
     4=&ColorFromCurve(
      ColorCurve=DataInterface() {propertyValues {
        AlphaCurve {keys [<at 0 2>,<at 0.8992 0>]}
        BlueCurve {keys [<at 0.0023 0.2292>,<at 0.2231 0.2292>,<at 0.6127 0.2292>]}
        GreenCurve {keys [<at 0.0023 0.0458>,<at 0.2231 0.0757>,<at 0.6127 0.0917>]}
        RedCurve {keys [<at 0.0023 0.1986>,<at 0.2231 0.2164>,<at 0.6127 0.2292>]}
       }}
      CurveIndex=$Particles.Update Flash Star Age
     )
     5=&ColorFromCurve(
      ColorCurve=DataInterface() {propertyValues {
        AlphaCurve {keys [<at 0 2>,<at 0.8992 0>]}
        BlueCurve {keys [<at 0.0023 0.087>,<at 0.2231 0.087>,<at 0.6127 0.087>]}
        GreenCurve {keys [<at 0.0023 0.0009>,<at 0.2231 0.0009>,<at 0.6127 0.0009>]}
        RedCurve {keys [<at 0.0023 0.0152>,<at 0.2231 0.0296>,<at 0.6127 0.0439>]}
       }}
      CurveIndex=$Particles.Update Flash Star Age
     )
     Index=&MakeIntFromFloat(Float=$User.Variant)
     Num=::NewEnumerator4
    )
    Particles.Update Flash Star Color Alt:LinearColor=&MergeHelper_ColorSwitch(
     0=&ColorFromCurve(
      ColorCurve=DataInterface() {propertyValues {
        AlphaCurve {keys [<at 0 2>,<at 0.8992 0>]}
        BlueCurve {keys [<at 0.1978 0>]}
        GreenCurve {keys [<at 0.1978 0>]}
        RedCurve {keys [<at 0.1978 0>]}
       }}
      CurveIndex=$Particles.Update Flash Star Age
     )
     1=&ColorFromCurve(
      ColorCurve=DataInterface() {propertyValues {
        AlphaCurve {keys [<at 0 2>,<at 0.8992 0>]}
        BlueCurve {keys [<at 0.1978 0>,<at 0.4068 0.0024>,<at 0.6127 0.0003>]}
        GreenCurve {keys [<at 0.1978 0.0799>,<at 0.4068 0.0399>,<at 0.6127 0.0156>]}
        RedCurve {keys [<at 0.1978 0.0532>,<at 0.4068 0.0305>,<at 0.6127 0.0131>]}
       }}
      CurveIndex=$Particles.Update Flash Star Age
     )
     2=&ColorFromCurve(
      ColorCurve=DataInterface() {propertyValues {
        AlphaCurve {keys [<at 0 2>,<at 0.8992 0>]}
        BlueCurve {keys [<at 0.1978 0>,<at 0.4068 0.0098>,<at 0.6127 0.0011>]}
        GreenCurve {keys [<at 0.1978 0.3865>,<at 0.4068 0.1003>,<at 0.6127 0.0275>]}
        RedCurve {keys [<at 0.1978 0.5797>,<at 0.4068 0.1649>,<at 0.6127 0.0538>]}
       }}
      CurveIndex=$Particles.Update Flash Star Age
     )
     3=&ColorFromCurve(
      ColorCurve=DataInterface() {propertyValues {
        AlphaCurve {keys [<at 0 2>,<at 0.8992 0>]}
        BlueCurve {keys [<at 0.1978 1>,<at 0.4068 0.6551>,<at 0.6127 0.0538>]}
        GreenCurve {keys [<at 0.1978 0.8562>,<at 0.4068 0.5965>,<at 0.6127 0.0494>]}
        RedCurve {keys [<at 0.1978 0.1374>,<at 0.4068 0.1088>,<at 0.6127 0.0011>]}
       }}
      CurveIndex=$Particles.Update Flash Star Age
     )
     4=&ColorFromCurve(
      ColorCurve=DataInterface() {propertyValues {
        AlphaCurve {keys [<at 0 2>,<at 0.8992 0>]}
        BlueCurve {keys [<at 0.1978 3>,<at 0.4068 0.8958>,<at 0.6127 0.4219>]}
        GreenCurve {keys [<at 0.1978 1.5>,<at 0.4068 0.6852>,<at 0.6127 0.2153>]}
        RedCurve {keys [<at 0.1978 0>,<at 0.4068 0.0532>,<at 0.6127 0.0087>]}
       }}
      CurveIndex=$Particles.Update Flash Star Age
     )
     5=&ColorFromCurve(
      ColorCurve=DataInterface() {propertyValues {
        AlphaCurve {keys [<at 0 2>,<at 0.8992 0>]}
        BlueCurve {keys [<at 0.1978 0.3652>,<at 0.4068 0.6087>,<at 0.6127 0.2696>]}
        GreenCurve {keys [<at 0.1978 0.2542>,<at 0.4068 0.4>,<at 0.6127 0.1876>]}
        RedCurve {keys [<at 0.1978 0.0647>,<at 0.4068 0.1078>,<at 0.6127 0.0478>]}
       }}
      CurveIndex=$Particles.Update Flash Star Age
     )
     Index=&MakeIntFromFloat(Float=$User.Variant)
     Num=::NewEnumerator4
    )
    Particles.Update Flash Star Sprize Size:Vector2f=&Multiply_Vector2D_ByFloat(
     Float=&FloatFromCurve(
      FloatCurve=DataInterface() {propertyValues {Curve {keys [
          {arriveTangent=2.7013, interpMode=RCIM_Cubic, leaveTangent=2.7013, tangentMode=RCTM_User, time=0, value=0}
          {interpMode=RCIM_Cubic, tangentMode=RCTM_User, time=1, value=1}
         ]}}}
      CurveIndex=$Particles.Update Flash Star Age
     )
     Vector2D=$Particles.Initial.Flash Star Sprite Size
    )
    Particles.Update Flash Star Sprize Size Alt:Vector2f=&Multiply_Vector2D_ByFloat(
     Float=&FloatFromCurve(
      FloatCurve=DataInterface() {propertyValues {Curve {keys [
          {arriveTangent=2.7013, interpMode=RCIM_Cubic, leaveTangent=2.7013, tangentMode=RCTM_User, time=0, value=0}
          {interpMode=RCIM_Cubic, tangentMode=RCTM_User, time=1, value=1}
         ]}}}
      CurveIndex=$Particles.Update Flash Star Age
     )
     Vector2D=$Particles.Initial.Flash Star Sprite Size Alt
    )
   ) {isSetParameters=true}
   /SetVariables(Particles.Update Flash Flare Color:LinearColor=&MakeLinearColorFromVectorAndFloat(
     "Vector (RGB)"=&Multiply_Vector(
      A=&Multiply_Vector(
       A=&MakeVectorFromLinearColorRGB(Color=$Particles.Flash Flare Color)
       B=&MakeVectorFromLinearColorRGB(Color=&MergeHelper_ColorSwitch(
         0=&ColorFromCurve(ColorCurve=DataInterface() {propertyValues {
            AlphaCurve {keys [<at 0 1 RCIM_Cubic>,<at 1 1 RCIM_Cubic>]}
            BlueCurve {keys [<at 0 0.3663 RCIM_Cubic>,<at 0.7502 0.1267>]}
            GreenCurve {keys [<at 0 0.3663 RCIM_Cubic>,<at 0.7502 0.1267>]}
            RedCurve {keys [<at 0 0.3663 RCIM_Cubic>,<at 0.7502 0.1267>]}
           }})
         1=&ColorFromCurve(ColorCurve=DataInterface() {propertyValues {
            AlphaCurve {keys [<at 0 1 RCIM_Cubic>,<at 1 1 RCIM_Cubic>]}
            BlueCurve {keys [<at 0 0 RCIM_Cubic>,<at 0.713 0 RCIM_Cubic>]}
            GreenCurve {keys [<at 0 0.5 RCIM_Cubic>,<at 0.713 0.3 RCIM_Cubic>]}
            RedCurve {keys [<at 0 0.4583 RCIM_Cubic>,<at 0.713 0.225 RCIM_Cubic>]}
           }})
         2=&ColorFromCurve(ColorCurve=DataInterface() {propertyValues {
            AlphaCurve {keys [<at 0 1 RCIM_Cubic>,<at 1 1 RCIM_Cubic>]}
            BlueCurve {keys [<at 0 0.1942 RCIM_Cubic>,<at 0.713 0 RCIM_Cubic>]}
            GreenCurve {keys [<at 0 0.6643 RCIM_Cubic>,<at 0.713 0.4167 RCIM_Cubic>]}
            RedCurve {keys [<at 0 1 RCIM_Cubic>,<at 0.713 1 RCIM_Cubic>]}
           }})
         3=&ColorFromCurve(ColorCurve=DataInterface() {propertyValues {
            AlphaCurve {keys [<at 0 1 RCIM_Cubic>,<at 1 1 RCIM_Cubic>]}
            BlueCurve {keys [<at 0 1 RCIM_Cubic>,<at 0.713 0.8 RCIM_Cubic>]}
            GreenCurve {keys [<at 0 0.8333 RCIM_Cubic>,<at 0.713 0.5333 RCIM_Cubic>]}
            RedCurve {keys [<at 0 0 RCIM_Cubic>,<at 0.713 0 RCIM_Cubic>]}
           }})
         4=&ColorFromCurve(ColorCurve=DataInterface() {propertyValues {
            AlphaCurve {keys [<at 0 1 RCIM_Cubic>,<at 1 1 RCIM_Cubic>]}
            BlueCurve {keys [<at 0 1 RCIM_Cubic>,<at 0.713 0.8 RCIM_Cubic>]}
            GreenCurve {keys [<at 0 0.4667 RCIM_Cubic>,<at 0.713 0.2017 RCIM_Cubic>]}
            RedCurve {keys [<at 0 0 RCIM_Cubic>,<at 0.713 0.7003 RCIM_Cubic>]}
           }})
         5=&ColorFromCurve(ColorCurve=DataInterface() {propertyValues {
            AlphaCurve {keys [<at 0 1 RCIM_Cubic>,<at 1 1 RCIM_Cubic>]}
            BlueCurve {keys [<at 0 1 RCIM_Cubic>,<at 0.713 0.8 RCIM_Cubic>]}
            GreenCurve {keys [<at 0 0.1667 RCIM_Cubic>,<at 0.713 0.5806 RCIM_Cubic>]}
            RedCurve {keys [<at 0 0 RCIM_Cubic>,<at 0.713 0.2017 RCIM_Cubic>]}
           }})
         Index=&MakeIntFromFloat(Float=$User.Variant)
         Num=::NewEnumerator4
        ))
      )
      B=&VectorFromFloat(Value=&FN_FrontEndFloatModifier(Gameplay=1, Front End=1))
     )
     "Float (Alpha)"=&Multiply_Float(
      A=&Multiply_Float(
       A=&MakeFloatFromLinearColor(LinearColor=$System.TODColorMultiplier, Channel:ENiagaraLinearColor_Channels=::NewEnumerator3)
       B=&SelectFloatFromArray(
        Direct Array Index=&MakeIntFromFloat(Float=$User.Variant)
        Float Selection Array=DataInterface() {propertyValues {FloatData [1,1,1,1,1,1]}}
       )
      )
      B=&Lerp_Float(A=1, B=0, Alpha=$Particles.Update Flash Flare Age)
     )
    )) {isSetParameters=true}
   &FN_Light_Attributes(
    Color=&MergeHelper_ColorSwitch(
     0=<rgba 4 4 4 1>
     1=<rgba 7.3667 8 0.4 1>
     2=<rgba 15 8.4375 3.75 1>
     3=<rgba 0.75 11.4375 15 1>
     4=<rgba 2.087 7.4674 15 1>
     5=<rgba 2.087 4.2391 15 1>
     Index=&MakeIntFromFloat(Float=$User.Variant)
     Num=::NewEnumerator4
    )
    Create Color Attribute=true
    Create Exponent Attribute=true
    Create Radius Attribute=true
    Exponent=4
    Radius=75
    Scale Color=&FloatFromCurve(FloatCurve=DataInterface() {propertyValues {Curve {keys [
         {interpMode=RCIM_Cubic, tangentMode=RCTM_User, time=0, value=1}
         {arriveTangent=-2.0035, interpMode=RCIM_Cubic, leaveTangent=-2.0035, tangentMode=RCTM_User, time=1, value=0}
        ]}}})
   )
   &FN_Sprite_Attributes(
    Color=&ColorFromCurve(ColorCurve=DataInterface() {propertyValues {
       AlphaCurve {keys [<at 0 2.5 RCIM_Cubic>,<at 1 0 RCIM_Cubic>]}
       BlueCurve {keys [<at 0 0 RCIM_Cubic>,<at 0.9989 0>]}
       GreenCurve {keys [<at 0 0 RCIM_Cubic>,<at 0.9989 0>]}
       RedCurve {keys [<at 0 0 RCIM_Cubic>,<at 0.9989 0>]}
      }})
    Create Color Attribute=true
    Create SpriteSize Attribute=true
    SpriteSize=&Vector2DFromFloat(Value=&FloatFromCurve(
      FloatCurve=DataInterface() {propertyValues {Curve {keys [
          {arriveTangent=7.1084, interpMode=RCIM_Cubic, leaveTangent=7.1084, tangentMode=RCTM_User, time=0, value=0.375}
          {arriveTangent=-2.821, interpMode=RCIM_Cubic, leaveTangent=-2.821, tangentMode=RCTM_User, time=0.5, value=0.625}
          {tangentMode=RCTM_User, time=1, value=0}
         ]}}}
      Scale Curve=&Multiply_Float(A=100, B=&FN_FrontEndFloatModifier(Gameplay=1, Front End=$System.FrontendScale))
     ))
    Write to Module Namespace=true
   )
   &FN_Sprite_Attributes(
    Camera Offset:NiagaraFloat=&FN_FrontEndFloatModifier(Gameplay=15, Front End=0)
    Color=&MultiplyLinearColor(
     Linear Color=<rgba 2 1.3708 0.9797 1>
     Scale Factor=&MakeVector4(
      X=1
      Y=1
      Z=1
      W=&Multiply_Float(
       A=&FloatFromCurve(
        FloatCurve=DataInterface() {propertyValues {Curve {keys [
            {arriveTangent=9.0725, interpMode=RCIM_Cubic, leaveTangent=9.0725, tangentMode=RCTM_User, time=0, value=0}
            {arriveTangent=-2.5249, interpMode=RCIM_Cubic, leaveTangent=-2.5249, tangentMode=RCTM_User, time=0.1408, value=0.9847}
            {arriveTangent=-0.0405, interpMode=RCIM_Cubic, leaveTangent=-0.0405, tangentMode=RCTM_User, time=1, value=0}
           ]}}}
        CurveIndex=&Multiply_Float(A=$Particles.NormalizedAge, B=1.1)
        Scale Curve=&FN_FrontEndFloatModifier(Gameplay=1, Front End=0.2)
       )
       B=&FNTimeOfDayScalar(Night Scale=0.3, Morning Scale=0.5, Evening Scale=0.5)
      )
     )
    )
    Create Camera Offset Attribute=true
    Create Color Attribute=true
    Create SpriteSize Attribute=true
    SpriteSize=&Vector2DFromFloat(Value=&FloatFromCurve(
      FloatCurve=DataInterface() {propertyValues {Curve {keys [{arriveTangent=0.8072, interpMode=RCIM_Cubic, leaveTangent=0.8072, tangentMode=RCTM_User, time=0, value=0.7}, <at 1 1.075>]}}}
      CurveIndex=&Multiply_Float(A=$Particles.NormalizedAge, B=1.1)
      Scale Curve=&Multiply_Float(A=90, B=&FN_FrontEndFloatModifier(Gameplay=1, Front End=$System.FrontendScale))
     ))
    Write to Module Namespace=true
   ) {module=FN_Sprite_Attributes001}
   &FN_Sprite_Attributes(
    Alignment:Vector3f=&FN_FrontEndVector3Modifier(
     Gameplay=<0 1 0>
     Front End=&Multiply_Vector(A=&ReturnCameraVector(Camera Vectors=::NewEnumerator3, Coordinate Space=::Local), B=<-1 -1 -1>)
    )
    Camera Offset:NiagaraFloat=&FN_FrontEndFloatModifier(Gameplay=20, Front End=5)
    Color=&MultiplyLinearColor(
     Linear Color=&Multiply_LinearColorByFloat(
      Linear Color=<rgba 1 0.8783 0.9952 1>
      Float=&FloatFromCurve(
       FloatCurve=DataInterface() {propertyValues {Curve {keys [
           {interpMode=RCIM_Cubic, tangentMode=RCTM_User, time=0, value=1}
           {arriveTangent=-0.6223, interpMode=RCIM_Cubic, leaveTangent=-0.6223, tangentMode=RCTM_User, time=1, value=0.6}
          ]}}}
       CurveIndex=$Particles.HoloLifetime
       Scale Curve=&FNTimeOfDayScalar(Night Scale=0.75, Morning Scale=0.95, Evening Scale=0.95)
      )
     )
     Scale Factor=&MakeVector4(
      X=1
      Y=1
      Z=1
      W=&FloatFromCurve(
       FloatCurve=DataInterface() {propertyValues {Curve {keys [
           {arriveTangent=-0.0319, interpMode=RCIM_Cubic, leaveTangent=-0.0319, tangentMode=RCTM_User, time=0, value=1}
           {arriveTangent=-1.1464, interpMode=RCIM_Cubic, leaveTangent=-1.1464, tangentMode=RCTM_User, time=1, value=0}
          ]}}}
       CurveIndex=$Particles.HoloLifetime
       Scale Curve=0.5
      )
     )
    )
    Create Alignment Attribute=true
    Create Camera Offset Attribute=true
    Create Color Attribute=true
    Create Facing Attribute=true
    Create SpriteSize Attribute=true
    Facing:Vector3f=&FN_FrontEndVector3Modifier(
     Gameplay=<1 0 0>
     Front End=&Multiply_Vector(A=&ReturnCameraVector(Camera Vectors=::NewEnumerator1, Coordinate Space=::Local), B=<-1 -1 -1>)
    )
    SpriteSize=&Vector2DFromFloat(Value=&FloatFromCurve(
      FloatCurve=DataInterface() {propertyValues {Curve {keys [
          {arriveTangent=1.048, interpMode=RCIM_Cubic, leaveTangent=1.048, tangentMode=RCTM_User, time=0, value=0.7}
          {arriveTangent=0.0147, interpMode=RCIM_Cubic, leaveTangent=0.0147, tangentMode=RCTM_User, time=1, value=1}
         ]}}}
      CurveIndex=$Particles.HoloLifetime
      Scale Curve=&Multiply_Float(A=100, B=&FN_FrontEndFloatModifier(Gameplay=1, Front End=$System.FrontendScale))
     ))
    Write to Module Namespace=true
   ) {module=FN_Sprite_Attributes002}
   &FN_Sprite_Attributes(
    Camera Offset:NiagaraFloat=&FN_FrontEndFloatModifier(Gameplay=7, Front End=12)
    Color=&Multiply_LinearColorByFloat(
     Linear Color=&MakeLinearColorFromVectorAndFloat(
      "Vector (RGB)"=&MakeVectorFromLinearColorRGB(Color=<rgba 0.209 0.9868 1 0.5>)
      "Float (Alpha)"=&RandomRangeFloat(Minimum=0.3, Maximum=0.4)
     )
     Float=&FloatFromCurve(
      FloatCurve=DataInterface() {propertyValues {Curve {keys [
          {arriveTangent=-0.0405, interpMode=RCIM_Cubic, leaveTangent=-0.0405, tangentMode=RCTM_User, time=0, value=1}
          {arriveTangent=-2.127, interpMode=RCIM_Cubic, leaveTangent=-2.127, tangentMode=RCTM_User, time=1, value=0}
         ]}}}
      CurveIndex=$Particles.Update GalaxySmoke Lifetime
     )
    )
    Create Camera Offset Attribute=true
    Create Color Attribute=true
    Create Sprite Rotation Attribute=true
    Create SpriteSize Attribute=true
    SpriteRotation:NiagaraFloat=&RandomRangeFloat(Minimum=-360, Maximum=360)
    SpriteSize=&Vector2DFromFloat(Value=&FloatFromCurve(
      FloatCurve=DataInterface() {propertyValues {Curve {keys [
          {arriveTangent=1.8336, interpMode=RCIM_Cubic, leaveTangent=1.8336, tangentMode=RCTM_User, time=0, value=0}
          {interpMode=RCIM_Cubic, tangentMode=RCTM_User, time=0.9, value=1}
         ]}}}
      CurveIndex=$Particles.Update GalaxySmoke Lifetime
      Scale Curve=&Multiply_Float(
       A=&RandomRangeFloat(Minimum=100, Maximum=110)
       B=&FN_FrontEndFloatModifier(Gameplay=1, Front End=$System.FrontendScale)
      )
     ))
    Write to Module Namespace=true
   ) {module=FN_Sprite_Attributes003}
   /SetVariables(Particles.HoloDyn:Vector4f=&MakeVector4(
     X=&FloatFromCurve(
      FloatCurve=DataInterface() {propertyValues {Curve {keys [
          {arriveTangent=1.5862, interpMode=RCIM_Cubic, leaveTangent=1.5862, tangentMode=RCTM_User, time=0, value=0}
          {arriveTangent=0.159, interpMode=RCIM_Cubic, leaveTangent=0.159, tangentMode=RCTM_User, time=1, value=0.8}
         ]}}}
      CurveIndex=$Particles.HoloLifetime
     )
     Y=&FloatFromCurve(
      FloatCurve=DataInterface() {propertyValues {Curve {keys [
          {arriveTangent=2.5938, interpMode=RCIM_Cubic, leaveTangent=2.5938, tangentMode=RCTM_User, time=0, value=0.2}
          {interpMode=RCIM_Cubic, tangentMode=RCTM_User, time=0.2, value=0.5}
          {tangentMode=RCTM_User, time=1, value=0}
         ]}}}
      CurveIndex=$Particles.HoloLifetime
      Scale Curve=0.7
     )
     Z=&FloatFromCurve(
      FloatCurve=DataInterface() {propertyValues {Curve {keys [{interpMode=RCIM_Cubic, tangentMode=RCTM_User, time=0, value=1}, {arriveTangent=-0.3137, tangentMode=RCTM_User, time=1, value=0.8}]}}}
      CurveIndex=$Particles.HoloLifetime
      Scale Curve=1.5
     )
     W=&Add_Float(
      A=&RandomRangeFloat(Minimum=0, Maximum=0.25)
      B=&FloatFromCurve(
       FloatCurve=DataInterface() {propertyValues {Curve {keys [{interpMode=RCIM_Cubic, tangentMode=RCTM_User, time=0, value=0}, {interpMode=RCIM_Cubic, tangentMode=RCTM_User, time=1, value=1}]}}}
       CurveIndex=$Particles.HoloLifetime
       Scale Curve=0.25
      )
     )
    )) {isSetParameters=true}
   /SetVariables(Particles.Dark_Glow=&MakeLinearColorFromVectorAndFloat(
     "Vector (RGB)"=&Multiply_Vector(
      A=&MakeVectorFromLinearColorRGB(Color=&MultiplyLinearColor(
        Linear Color=&ColorFromCurve(ColorCurve=DataInterface() {propertyValues {
           AlphaCurve {keys []}
           BlueCurve {keys [<at 0.2868 0.3281 RCIM_Cubic>,<at 0.5555 0.1406>,<at 0.9372 0.05 RCIM_Cubic>]}
           GreenCurve {keys [<at 0.2868 0.3281 RCIM_Cubic>,<at 0.5555 0.1406>,<at 0.9372 0.05 RCIM_Cubic>]}
           RedCurve {keys [<at 0.2868 0.3281 RCIM_Cubic>,<at 0.5555 0.1406>,<at 0.9372 0.05 RCIM_Cubic>]}
          }})
        Scale Factor=&MakeVector4_FromLinearColor(LinearColor=&SelectLinearColorFromArray(
          Array Sampling Mode=::NewEnumerator1
          Direct Array Index=&MakeIntFromFloat(Float=$User.Variant, Conversion Mode=::NewEnumerator1)
          Color Selection Array=DataInterface() {propertyValues {ColorData [
             <rgba 0.0999 0.0999 0.0999 1>
             <rgba 0.0251 0.026 0 1>
             <rgba 0.0781 0.0658 0.0332 1>
             <rgba 0.0385 0.0788 0.099 0.4783>
             <rgba 0.0399 0.0819 0.099 0.5826>
             <rgba 0.3385 0.1673 0.3317 0.4261>
            ]}}
          Random Seed=?
         ))
       ))
      B=&MakeVectorFromLinearColorRGB(Color=$System.TODColorMultiplier)
     )
     "Float (Alpha)"=&Multiply_Float(
      A=&MakeFloatFromLinearColor(LinearColor=$System.TODColorMultiplier, Channel:ENiagaraLinearColor_Channels=::NewEnumerator3)
      B=&Multiply_Float(
       A=&FN_FrontEndFloatModifier(
        Gameplay=&MergeHelper_FloatSwitch(0=1, 1=1, 2=1, 3=1, 4=1, 5=0.7, Index=&MakeIntFromFloat(Float=$User.Variant), Num=::NewEnumerator4)
        Front End=0.5
       )
       B=&FloatFromCurve(FloatCurve=DataInterface() {propertyValues {Curve {keys [
            {arriveTangent=-0.3275, interpMode=RCIM_Cubic, leaveTangent=-0.3275, tangentMode=RCTM_User, time=0, value=1}
            {arriveTangent=-0.3009, interpMode=RCIM_Cubic, leaveTangent=-0.3009, tangentMode=RCTM_User, time=1, value=0}
           ]}}})
      )
     )
    )) {isSetParameters=true}
   /SetVariables(Particles.Dark_Glow_Size=&Multiply_Vector2(
     A=&Vector2DFromFloat(Value=130)
     B=&Vector2DFromFloat(Value=&FloatFromCurve(
       FloatCurve=DataInterface() {propertyValues {Curve {keys [
           {arriveTangent=4.1378, interpMode=RCIM_Cubic, leaveTangent=4.1378, tangentMode=RCTM_User, time=0, value=0}
           {interpMode=RCIM_Cubic, leaveTangent=0.5, time=1, value=1.45}
          ]}}}
       Scale Curve=&FN_FrontEndFloatModifier(Gameplay=0.75, Front End=0.4)
      ))
    )) {isSetParameters=true}
   &FN_Mesh_Attributes(
    Color=&MakeLinearColorFromVectorAndFloat(
     "Vector (RGB)"=&Multiply_Vector(
      A=&Lerp_Vector(
       A=&MakeVectorFromLinearColorRGB(Color=<rgba 1 0.1 0.418 1>)
       B=&MakeVectorFromLinearColorRGB(Color=<rgba 0 0.2294 0.974 1>)
       Alpha=&FloatFromCurve(
        FloatCurve=DataInterface() {propertyValues {Curve {keys [{interpMode=RCIM_Cubic, tangentMode=RCTM_User, time=0.4, value=0}, {interpMode=RCIM_Cubic, tangentMode=RCTM_User, time=0.6, value=1}]}}}
        CurveIndex=$Particles.Update GalaxyWave Lifetime
       )
      )
      B=&VectorFromFloat(Value=&FloatFromCurve(
        FloatCurve=DataInterface() {propertyValues {Curve {keys [
            {arriveTangent=11.5613, interpMode=RCIM_Cubic, leaveTangent=11.5613, tangentMode=RCTM_User, time=0, value=0}
            {interpMode=RCIM_Cubic, tangentMode=RCTM_User, time=0.2, value=1}
            {tangentMode=RCTM_User, time=1, value=0}
           ]}}}
        CurveIndex=$Particles.Update GalaxyWave Lifetime
        Scale Curve=3
       ))
     )
     "Float (Alpha)"=&Multiply_Float(
      A=&FloatFromCurve(
       FloatCurve=DataInterface() {propertyValues {Curve {keys [{interpMode=RCIM_Cubic, tangentMode=RCTM_User, time=0, value=1}, {interpMode=RCIM_Cubic, tangentMode=RCTM_User, time=1, value=0}]}}}
       CurveIndex=$Particles.Update GalaxyWave Lifetime
      )
      B=&FNTimeOfDayScalar(Night Scale=0.7, Morning Scale=0.5, Evening Scale=0.5)
     )
    )
    Create Color Attribute=true
    Create Position Attribute=true
    Create Scale Attribute=true
    Scale=&Multiply_Vector(
     A=&VectorFromFloat(Value=0.7)
     B=&VectorFromFloat(Value=&FloatFromCurve(
       FloatCurve=DataInterface() {propertyValues {Curve {keys [
           {arriveTangent=1.152, interpMode=RCIM_Cubic, leaveTangent=1.152, tangentMode=RCTM_User, time=0, value=0.65}
           {arriveTangent=0.0224, interpMode=RCIM_Cubic, leaveTangent=0.0224, tangentMode=RCTM_User, time=1, value=1}
          ]}}}
       CurveIndex=$Particles.Update GalaxyWave Lifetime
       Scale Curve=&FN_FrontEndFloatModifier(Gameplay=1, Front End=$System.FrontendScale)
      ))
    )
    Write to Module Namespace=true
   )
   &FN_Mesh_Attributes(
    Color=&MakeLinearColorFromVectorAndFloat(
     "Vector (RGB)"=&Multiply_Vector(
      A=&MakeVectorFromLinearColorRGB(Color=&ColorFromCurve(ColorCurve=DataInterface() {propertyValues {
          AlphaCurve {keys [<at 0 1>,<at 0.3576 1>,<at 1 0>]}
          BlueCurve {keys [<at 0 0>,<at 0.1684 0>,<at 0.4403 0>]}
          GreenCurve {keys [<at 0 0.9167>,<at 0.1684 1>,<at 0.4403 1>]}
          RedCurve {keys [<at 0 1>,<at 0.1684 1>,<at 0.4403 0.8333>]}
         }}))
      B=&VectorFromFloat(Value=&FloatFromCurve(FloatCurve=DataInterface() {propertyValues {Curve {keys [<at 0 10>,<at 0.15 3>,<at 1 1>]}}}))
     )
     "Float (Alpha)"=&Multiply_Float(
      A=&FloatFromCurve(FloatCurve=DataInterface() {propertyValues {Curve {keys [{interpMode=RCIM_Cubic, tangentMode=RCTM_User, time=0, value=1}, {interpMode=RCIM_Cubic, tangentMode=RCTM_User, time=1, value=0}]}}})
      B=&FNTimeOfDayScalar(Night Scale=0.7, Morning Scale=0.5, Evening Scale=0.5)
     )
    )
    Create Color Attribute=true
    Create Position Attribute=true
    Create Scale Attribute=true
    Scale=&Multiply_Vector(
     A=&VectorFromFloat(Value=0.5)
     B=&VectorFromFloat(Value=&FloatFromCurve(
       FloatCurve=DataInterface() {propertyValues {Curve {keys [<at 0 0>,<at 0.1 1>,<at 1 2>]}}}
       Scale Curve=&FN_FrontEndFloatModifier(Gameplay=0.85, Front End=&Multiply_Float(A=0.8, B=$System.FrontendScale))
      ))
    )
    Write to Module Namespace=true
   ) {module=FN_Mesh_Attributes001}
   /SetVariables(
    Particles.GalaxyCamOffset:NiagaraFloat=&FN_FrontEndFloatModifier(Gameplay=25, Front End=0)
    Particles.GalaxyDynamicPar:Vector4f=&MakeVector4(
     X=&FloatFromCurve(
      FloatCurve=DataInterface() {propertyValues {Curve {keys [{arriveTangent=-2.1015, interpMode=RCIM_Cubic, leaveTangent=-2.1015, tangentMode=RCTM_User, time=0, value=0.5}, <at 1 -0.15>]}}}
      CurveIndex=$Particles.Update GalaxyWave Lifetime
     )
     Y=&FloatFromCurve(
      FloatCurve=DataInterface() {propertyValues {Curve {keys [{arriveTangent=-0.3464, interpMode=RCIM_Cubic, leaveTangent=-0.3464, tangentMode=RCTM_User, time=0, value=1}, <at 1 0.9>]}}}
      CurveIndex=$Particles.Update GalaxyWave Lifetime
     )
     Z=&FloatFromCurve(
      FloatCurve=DataInterface() {propertyValues {Curve {keys [{interpMode=RCIM_Cubic, tangentMode=RCTM_User, time=0, value=1}, {interpMode=RCIM_Cubic, tangentMode=RCTM_User, time=1, value=0.25}]}}}
      CurveIndex=$Particles.Update GalaxyWave Lifetime
     )
     W=0
    )
   ) {isSetParameters=true}
   /SetVariables(Particles.ToxicWaveDynamicPar:Vector4f=&MakeVector4(
     X=&FloatFromCurve(FloatCurve=DataInterface() {propertyValues {Curve {keys [{arriveTangent=-2.1015, interpMode=RCIM_Cubic, leaveTangent=-2.1015, tangentMode=RCTM_User, time=0, value=0.5}, <at 1 -0.15>]}}})
     Y=&FloatFromCurve(FloatCurve=DataInterface() {propertyValues {Curve {keys [{arriveTangent=-0.3464, interpMode=RCIM_Cubic, leaveTangent=-0.3464, tangentMode=RCTM_User, time=0, value=1}, <at 1 0.9>]}}})
     Z=&FloatFromCurve(FloatCurve=DataInterface() {propertyValues {Curve {keys [<at 0 0>,<at 0.15 1>,<at 1 1.2>]}}})
     W=0
    )) {isSetParameters=true}
  ]
  rendererClasses [
   &NiagaraSpriteRendererProperties
   &NiagaraLightRendererProperties
   &NiagaraMeshRendererProperties
  ]
  renderers [
   {rendererClass=&NiagaraSpriteRendererProperties}
   {rendererClass=&NiagaraSpriteRendererProperties, rendererIndex=1}
   {rendererClass=&NiagaraSpriteRendererProperties, rendererIndex=2}
   {rendererClass=&NiagaraSpriteRendererProperties, rendererIndex=3}
   {rendererClass=&NiagaraLightRendererProperties, rendererIndex=4}
   {rendererClass=&NiagaraSpriteRendererProperties, rendererIndex=5}
   {rendererClass=&NiagaraSpriteRendererProperties, rendererIndex=6}
   {rendererClass=&NiagaraSpriteRendererProperties, rendererIndex=7}
   {rendererClass=&NiagaraSpriteRendererProperties, rendererIndex=8}
   {rendererClass=&NiagaraSpriteRendererProperties, rendererIndex=9}
   {rendererClass=&NiagaraMeshRendererProperties, rendererIndex=10}
   {rendererClass=&NiagaraMeshRendererProperties, rendererIndex=11}
  ]
 }
 Impact_Smoke {
  emitterUpdateScript [
   &EmitterState()
   &SpawnBurst_Instantaneous(
    Spawn Count=&RandomRangeInt(Minimum=2, Maximum=3)
    Use Spawn Probability=false
    UseLoopCountLimit=false
   )
  ]
  particleSpawnScript [
   &InitializeParticle(
    Color=&MultiplyLinearColor(
     Linear Color=<rgba 1 1 1 0.2>
     Scale Factor=&MakeVector4_FromLinearColor(LinearColor=$System.TODColorMultiplier)
    )
    Lifetime Max=0.4
    Lifetime Min=0.3
    Lifetime Mode=::NewEnumerator1
    Sprite Size Mode=::NewEnumerator3
    Uniform Sprite Size=&Multiply_Float(
     A=&FN_FrontEndFloatModifier(Gameplay=1, Front End=&Multiply_Float(A=0.7, B=$System.FrontendScale))
     B=&RandomRangeFloat(Minimum=100, Maximum=130)
    )
    UsePositionOffset=false
    Write Material Random=true
   )
   &CameraOffset(Camera Offset Amount=&RandomRangeFloat(Minimum=20, Maximum=25, Evaluation Type=::NewEnumerator1))
   &ShapeLocation(
    Hemisphere Distribution=<1 0.5>
    MaskSpawnGroup=false
    Override Seed=false
    Rotation Angle=90
    Rotation Mode=::NewEnumerator1
    Sphere Radius=&MakeCustomFloatFromBool(Bool=$User.bIsFrontEnd, True Float=0, False Float=10)
    UseOwnerScale=true
   )
   &AddVelocity(Velocity=&Multiply_VectorByFloat(Vector=$Particles.ShapeLocation.ShapeVector, Float=10))
  ]
  particleUpdateScript [
   &ParticleState()
   &ScaleSpriteSize(Uniform Curve Sprite Scale=DataInterface() {propertyValues {Curve {keys [
        {arriveTangent=1.7761, interpMode=RCIM_Cubic, leaveTangent=1.7761, tangentMode=RCTM_User, time=0, value=0}
        {interpMode=RCIM_Cubic, tangentMode=RCTM_User, time=1, value=1}
       ]}}})
   &ScaleColor(
    Scale Alpha=&FloatFromCurve(
     FloatCurve=DataInterface() {propertyValues {Curve {keys [
         {arriveTangent=2.4358, interpMode=RCIM_Cubic, leaveTangent=2.4358, tangentMode=RCTM_User, time=0, value=0}
         <at 0.25 0.5 RCIM_Cubic>
         <at 1 0 RCIM_Cubic>
        ]}}}
     Scale Curve=&Multiply_Float(A=2, B=&FN_FrontEndFloatModifier(Gameplay=0.75, Front End=0.75))
    )
    Scale RGB=&Multiply_Vector(
     A=&MakeVectorFromLinearColorRGB(Color=&MergeHelper_ColorSwitch(
       0=&ColorFromCurve(ColorCurve=DataInterface() {propertyValues {
          AlphaCurve {keys [<at 0 1 RCIM_Cubic>,<at 1 0 RCIM_Cubic>]}
          BlueCurve {keys [
            {arriveTangent=2.3968, interpMode=RCIM_Cubic, time=0.003, value=1}
            <at 0.1089 0.6615>
            {arriveTangent=-0.6824, interpMode=RCIM_Cubic, leaveTangent=-0.6824, time=0.4263, value=0.5853}
            <at 0.9891 0.0608>
           ]}
          GreenCurve {keys [
            {arriveTangent=-0.0399, interpMode=RCIM_Cubic, time=0.003, value=1}
            <at 0.1089 0.6615>
            {arriveTangent=-0.6824, interpMode=RCIM_Cubic, leaveTangent=-0.6824, time=0.4263, value=0.5853}
            <at 0.9891 0.0608>
           ]}
          RedCurve {keys [
            {arriveTangent=-2.1571, interpMode=RCIM_Cubic, time=0.003, value=1}
            <at 0.1089 0.6615>
            {arriveTangent=-0.6824, interpMode=RCIM_Cubic, leaveTangent=-0.6824, time=0.4263, value=0.5853}
            <at 0.9891 0.0608>
           ]}
         }})
       1=&ColorFromCurve(ColorCurve=DataInterface() {propertyValues {
          AlphaCurve {keys [<at 0 1 RCIM_Cubic>,<at 1 0 RCIM_Cubic>]}
          BlueCurve {keys [
            {arriveTangent=2.3968, interpMode=RCIM_Cubic, time=0.0651, value=0.0696}
            {arriveTangent=-0.0471, interpMode=RCIM_Cubic, leaveTangent=-0.0471, time=0.2603, value=0.0278}
            <at 0.8495 0.0326>
           ]}
          GreenCurve {keys [
            {arriveTangent=-0.0399, interpMode=RCIM_Cubic, time=0.0651, value=1}
            {arriveTangent=-1.1619, interpMode=RCIM_Cubic, leaveTangent=-1.1619, time=0.2603, value=0.2}
            <at 0.8495 0.0887>
           ]}
          RedCurve {keys [
            {arriveTangent=-2.1571, interpMode=RCIM_Cubic, time=0.0651, value=0.8449}
            {arriveTangent=-0.9577, interpMode=RCIM_Cubic, leaveTangent=-0.9577, time=0.2603, value=0.2}
            <at 0.8495 0.0938>
           ]}
         }})
       2=&ColorFromCurve(ColorCurve=DataInterface() {propertyValues {
          AlphaCurve {keys [<at 0 1 RCIM_Cubic>,<at 1 0 RCIM_Cubic>]}
          BlueCurve {keys [
            {arriveTangent=2.3968, interpMode=RCIM_Cubic, time=0.003, value=0.25}
            <at 0.1089 0.113>
            {arriveTangent=-0.0914, interpMode=RCIM_Cubic, leaveTangent=-0.0914, time=0.4263, value=0.0278}
            <at 0.9891 0.0326>
           ]}
          GreenCurve {keys [
            {arriveTangent=-0.0399, interpMode=RCIM_Cubic, time=0.003, value=0.6875}
            <at 0.1089 0.5565>
            {arriveTangent=-0.5662, interpMode=RCIM_Cubic, leaveTangent=-0.5662, time=0.4263, value=0.1053}
            <at 0.9891 0.0581>
           ]}
          RedCurve {keys [
            {arriveTangent=-2.1571, interpMode=RCIM_Cubic, time=0.003, value=1}
            <at 0.1089 1>
            {arriveTangent=-1.0295, interpMode=RCIM_Cubic, leaveTangent=-1.0295, time=0.4263, value=0.2}
            <at 0.9891 0.0938>
           ]}
         }})
       3=&ColorFromCurve(ColorCurve=DataInterface() {propertyValues {
          AlphaCurve {keys [<at 0 1 RCIM_Cubic>,<at 1 0 RCIM_Cubic>]}
          BlueCurve {keys [
            {arriveTangent=2.3968, interpMode=RCIM_Cubic, time=0.003, value=1}
            <at 0.1089 1>
            {arriveTangent=-1.0295, interpMode=RCIM_Cubic, leaveTangent=-1.0295, time=0.4263, value=0.2}
            <at 0.9891 0.0938>
           ]}
          GreenCurve {keys [
            {arriveTangent=-0.0399, interpMode=RCIM_Cubic, time=0.003, value=1}
            <at 0.1089 0.9261>
            {arriveTangent=-0.9629, interpMode=RCIM_Cubic, leaveTangent=-0.9629, time=0.4263, value=0.1713}
            <at 0.9891 0.0785>
           ]}
          RedCurve {keys [
            {arriveTangent=-2.1571, interpMode=RCIM_Cubic, time=0.003, value=0.0948}
            <at 0.1089 0.113>
            {arriveTangent=-0.0914, interpMode=RCIM_Cubic, leaveTangent=-0.0914, time=0.4263, value=0.0278}
            <at 0.9891 0.0326>
           ]}
         }})
       4=&ColorFromCurve(ColorCurve=DataInterface() {propertyValues {
          AlphaCurve {keys [<at 0 1 RCIM_Cubic>,<at 1 0 RCIM_Cubic>]}
          BlueCurve {keys [{arriveTangent=2.3968, interpMode=RCIM_Cubic, time=0.0883, value=1.7519}, <at 0.4365 1>, <at 0.6989 0.0938>]}
          GreenCurve {keys [{arriveTangent=-0.0399, interpMode=RCIM_Cubic, time=0.0883, value=2}, <at 0.4365 0.2904>, <at 0.6989 0.0326>]}
          RedCurve {keys [{arriveTangent=-2.1571, interpMode=RCIM_Cubic, time=0.0883, value=0.1391}, <at 0.4365 0.113>, <at 0.6989 0.0836>]}
         }})
       5=&ColorFromCurve(ColorCurve=DataInterface() {propertyValues {
          AlphaCurve {keys [<at 0 1 RCIM_Cubic>,<at 1 0 RCIM_Cubic>]}
          BlueCurve {keys [{arriveTangent=2.3968, interpMode=RCIM_Cubic, time=0.003, value=1}, <at 0.1089 1>, <at 0.6958 0.0938>]}
          GreenCurve {keys [{arriveTangent=-0.0399, interpMode=RCIM_Cubic, time=0.003, value=0.2246}, <at 0.1089 0.6748>, <at 0.6958 0.0326>]}
          RedCurve {keys [{arriveTangent=-2.1571, interpMode=RCIM_Cubic, time=0.003, value=0.0696}, <at 0.1089 0.113>, <at 0.6958 0.053>]}
         }})
       Index=&MakeIntFromFloat(Float=$User.Variant)
       Num=::NewEnumerator4
      ))
     B=&VectorFromFloat(Value=&FN_FrontEndFloatModifier(Gameplay=1, Front End=2))
    )
    ScaleA=true
    ScaleRGB=true
   )
   &DynamicMaterialParameters(
    Index 0 Param 1=&Add_Float(
     A=&RandomRangeFloat(Minimum=0, Maximum=1)
     B=&FloatFromCurve(
      FloatCurve=DataInterface() {propertyValues {Curve {keys [
          <at 0 0 RCIM_Cubic>
          {arriveTangent=1.5947, interpMode=RCIM_Cubic, leaveTangent=1.5947, tangentMode=RCTM_User, time=0.2557, value=0.5257}
          <at 1 1 RCIM_Cubic>
         ]}}}
      CurveIndex=$Particles.Age
      Scale Curve=0.25
     )
    )
    Module.Param0WriteEnabled=true
    Module.Param1WriteEnabled=true
    Module.Param2WriteEnabled=true
    Module.Param3WriteEnabled=true
   ) {module=DynamicMaterialParameters001}
   &SolveForcesAndVelocity(Clamp Velocity=false, Limit Acceleration=false)
  ]
  rendererClasses [&NiagaraSpriteRendererProperties]
  renderers [
   {rendererClass=&NiagaraSpriteRendererProperties}
   {rendererClass=&NiagaraSpriteRendererProperties, rendererIndex=1}
  ]
 }
 Sparks {
  emitterUpdateScript [
   &EmitterState()
   &SpawnBurst_Instantaneous(
    Spawn Count=&MakeCustomIntFromBool(Bool=$User.bIsFrontEndPreview, True INT=30, False INT=&RandomRangeInt(Minimum=15, Maximum=20))
    Use Spawn Probability=false
    UseLoopCountLimit=false
   )
   &SpawnBurst_Instantaneous(
    Spawn Count=30
    Spawn Group=1
    Spawn Probability=&MakeCustomFloatFromBool(Bool=$User.bIsFrontEndPreview, True Float=0, False Float=1)
    Use Spawn Probability=true
    UseLoopCountLimit=false
   ) {module=SpawnBurst_Instantaneous001}
  ]
  particleSpawnScript [
   &InitializeParticle(
    Color=&MergeHelper_ColorSwitch(
     0=&SelectLinearColorFromArray(
      Array Sampling Mode=::NewEnumerator0
      Direct Array Index=?
      Color Selection Array=DataInterface() {propertyValues {ColorData [<rgba 0 0 0 1>, <rgba 1 1 1 1>]}}
      Random Seed=0
     )
     1=&SelectLinearColorFromArray(
      Array Sampling Mode=::NewEnumerator0
      Direct Array Index=?
      Color Selection Array=DataInterface() {propertyValues {ColorData [<rgba 0.0621 0.0677 0 3>, <rgba 1 1 0 1>]}}
      Random Seed=0
     )
     2=&SelectLinearColorFromArray(
      Array Sampling Mode=::NewEnumerator0
      Direct Array Index=?
      Color Selection Array=DataInterface() {propertyValues {ColorData [<rgba 0.0677 0.0451 0 3>, <rgba 0.7884 0.4586 0.0686 1>]}}
      Random Seed=0
     )
     3=&SelectLinearColorFromArray(
      Array Sampling Mode=::NewEnumerator0
      Direct Array Index=?
      Color Selection Array=DataInterface() {propertyValues {ColorData [<rgba 0.0311 0.1264 0.1739 3>, <rgba 0.2537 0.6102 0.7884 1>]}}
      Random Seed=0
     )
     4=&SelectLinearColorFromArray(
      Array Sampling Mode=::NewEnumerator0
      Direct Array Index=?
      Color Selection Array=DataInterface() {propertyValues {ColorData [<rgba 0.0166 0.3338 0.6377 3>, <rgba 0.8349 0.2947 0.8841 1>]}}
      Random Seed=0
     )
     5=&SelectLinearColorFromArray(
      Array Sampling Mode=::NewEnumerator0
      Direct Array Index=?
      Color Selection Array=DataInterface() {propertyValues {ColorData [<rgba 0.0693 0.1762 0.4522 3>, <rgba 0.136 0.0334 0.2681 1>]}}
      Random Seed=0
     )
     Index=&MakeIntFromFloat(Float=$User.Variant)
     Num=::NewEnumerator4
    )
    Lifetime Max=0.45
    Lifetime Min=0.05
    Lifetime Mode=::NewEnumerator1
    Sprite Size Mode=::NewEnumerator3
    Uniform Sprite Size=&Multiply_Float(
     A=&FN_FrontEndFloatModifier(Gameplay=1, Front End=$System.FrontendScale)
     B=&MergeHelper_FloatSwitch(
      0=&RandomRangeFloat(Minimum=3, Maximum=5)
      1=&RandomRangeFloat(Minimum=4.5, Maximum=5)
      2=?
      3=?
      Index=$Emitter.SpawnGroup
      Num=::NewEnumerator0
     )
    )
    UsePositionOffset=false
    Write Material Random=true
   )
   &ShapeLocation(
    MaskSpawnGroup=false
    Override Seed=false
    Sphere Radius=&Multiply_Float(A=2, B=&FN_FrontEndFloatModifier(Gameplay=0, Front End=$System.FrontendScale))
    UseOwnerScale=true
   )
   &ShapeLocation(
    Cylinder Height=&MergeHelper_FloatSwitch(0=0, 1=80, 2=?, 3=?, Index=$Emitter.SpawnGroup, Num=::NewEnumerator0)
    Cylinder Radius=&MergeHelper_FloatSwitch(0=0, 1=2, 2=?, 3=?, Index=$Emitter.SpawnGroup, Num=::NewEnumerator0)
    MaskSpawnGroup=false
    Override Seed=false
    Rotation Axis=$System.DirectionVector
    Rotation Mode=::NewEnumerator1
    Shape Primitive=::NewEnumerator1
    Surface Only=false
    UseOwnerScale=true
   ) {module=ShapeLocation001}
   &CameraOffset(Camera Offset Amount=&MergeHelper_FloatSwitch(0=20, 1=25, 2=?, 3=?, Index=$Emitter.SpawnGroup, Num=::NewEnumerator0))
   &AddVelocity(
    Cone Angle=&FN_FrontEndFloatModifier(Gameplay=180, Front End=360)
    Module.UseConeFalloff=false
    Module.UseVelDistribution=false
    Override Seed=false
    Velocity Mode=::NewEnumerator2
    Velocity Speed=&Multiply_Float(
     A=&RandomRangeFloat(Minimum=400, Maximum=750)
     B=&FN_FrontEndFloatModifier(
      Gameplay=&MergeHelper_FloatSwitch(0=1, 1=0, 2=?, 3=?, Index=$Emitter.SpawnGroup, Num=::NewEnumerator0)
      Front End=0
     )
    )
   )
   &AddVelocityFromPoint(
    Module.Use Falloff Distance=false
    Velocity Strength=&Multiply_Float(
     A=&RandomRangeFloat(Minimum=150, Maximum=200)
     B=&FN_FrontEndFloatModifier(
      Gameplay=&MergeHelper_FloatSwitch(0=0, 1=1.5, 2=?, 3=?, Index=$Emitter.SpawnGroup, Num=::NewEnumerator0)
      Front End=1.75
     )
    )
   )
   &AddVelocity(
    Velocity=$System.DirectionVector
    Velocity Speed Scale=&Multiply_Float(
     A=&RandomRangeFloat(Minimum=300, Maximum=700)
     B=&FN_FrontEndFloatModifier(
      Gameplay=&MergeHelper_FloatSwitch(0=0, 1=1, 2=?, 3=?, Index=$Emitter.SpawnGroup, Num=::NewEnumerator0)
      Front End=0
     )
    )
   ) {module=AddVelocity001}
  ]
  particleUpdateScript [
   &ParticleState()
   &AccelerationForce(Acceleration=<0 0 -500>)
   &Drag(Drag=&RandomRangeFloat(Minimum=5, Maximum=8), Use Linear Drag=true, Use Rotational Drag=true)
   &SolveForcesAndVelocity(Clamp Velocity=false, Limit Acceleration=false)
   &ScaleSpriteSizeBySpeed(
    Max Scale Factor=&MergeHelper_Vector2Switch(0=<0.5 10>, 1=<0.2 5>, 2=?, 3=?, 4=?, Index=$Emitter.SpawnGroup, Num=::NewEnumerator0)
    Min Scale Factor=&MergeHelper_Vector2Switch(0=<0.5 1>, 1=<0.2 0.2>, 2=?, 3=?, 4=?, Index=$Emitter.SpawnGroup, Num=::NewEnumerator0)
   )
   &ScaleSpriteSize(
    Uniform Curve Scale=&Multiply_Float(A=1, B=&FN_FrontEndFloatModifier(Gameplay=1, Front End=0.5))
    Uniform Curve Sprite Scale=DataInterface() {propertyValues {Curve {keys [{arriveTangent=-6.725, interpMode=RCIM_Cubic, leaveTangent=-6.725, tangentMode=RCTM_User, time=0.65, value=1}, <at 1 0 RCIM_Cubic>]}}}
   )
   &ScaleColor(
    Scale Alpha=&MakeFloatFromLinearColor(LinearColor=$System.TODColorMultiplier, Channel:ENiagaraLinearColor_Channels=::NewEnumerator3)
    Scale RGB=&Multiply_Vector(
     A=&MakeVectorFromLinearColorRGB(Color=$System.TODColorMultiplier)
     B=&VectorFromFloat(Value=&FloatFromCurve(
       FloatCurve=DataInterface() {propertyValues {Curve {keys [
           {interpMode=RCIM_Cubic, tangentMode=RCTM_User, time=0, value=1}
           <at 0.3 0.2>
           {interpMode=RCIM_Cubic, tangentMode=RCTM_User, time=1, value=0}
          ]}}}
       Scale Curve=10
      ))
    )
    ScaleA=true
    ScaleRGB=true
   )
   /SetVariables(
    Particles.GlowColor=&MakeLinearColorFromVectorAndFloat(
     "Vector (RGB)"=&MakeVectorFromLinearColorRGB(Color=$Particles.Initial.Color)
     "Float (Alpha)"=&FloatFromCurve(FloatCurve=DataInterface() {propertyValues {Curve {keys [<at 0.5 1>,<at 1 0>]}}}, Scale Curve=0.08)
    )
    Particles.GlowSize=&Multiply_Vector2D(A=$Particles.SpriteSize, B=<8 2>)
   ) {isSetParameters=true}
  ]
  rendererClasses [&NiagaraSpriteRendererProperties]
  renderers [
   {rendererClass=&NiagaraSpriteRendererProperties}
   {rendererClass=&NiagaraSpriteRendererProperties, rendererIndex=1}
  ]
 }
 Toxic_Directional_Smoke {
  emitterUpdateScript [
   /SetVariables(Emitter.Spline_Base:NiagaraFloat=&Multiply_Float(
     A=&FloatFromCurve(
      FloatCurve=DataInterface() {propertyValues {Curve {keys [{arriveTangent=-5.7778, interpMode=RCIM_Cubic, time=0, value=1}, <at 0.35 1>]}}}
      CurveIndex=$Emitter.NormalizedLoopAge
     )
     B=&FN_FrontEndFloatModifier(Gameplay=60, Front End=40)
    )) {isSetParameters=true}
   /SetVariables(Emitter.Spline_Tip:NiagaraFloat=&Multiply_Float(
     A=&FloatFromCurve(
      FloatCurve=DataInterface() {propertyValues {Curve {keys [{arriveTangent=-5.7778, interpMode=RCIM_Cubic, time=0, value=1.5}, <at 0.35 0.5>]}}}
      CurveIndex=$Emitter.NormalizedLoopAge
     )
     B=&FN_FrontEndFloatModifier(Gameplay=30, Front End=15)
    )) {isSetParameters=true}
   &EmitterState()
   &SpawnBurst_Instantaneous(
    Spawn Probability=&FN_FrontEndFloatModifier(Gameplay=1, Front End=1)
    Use Spawn Probability=true
    UseLoopCountLimit=false
   ) {module=SpawnBurst_Instantaneous001}
   &SpawnBurst_Instantaneous(
    Spawn Group=1
    Spawn Probability=&FN_FrontEndFloatModifier(Gameplay=1, Front End=1)
    Use Spawn Probability=true
    UseLoopCountLimit=false
   )
  ]
  particleSpawnScript [
   /SetVariables(Particles.VisibilityTag=$Emitter.SpawnGroup) {isSetParameters=true}
   &InitializeParticle(
    Color=<rgba 2.5 2.5 2.5 25>
    Lifetime=&MergeHelper_FloatSwitch(0=0.35, 1=0.2, 2=?, 3=?, Index=$Emitter.SpawnGroup, Num=::NewEnumerator0)
    Mesh Scale=&RandomRangeVector(Minimum=<1.3 1 1>, Maximum=<1.5 1 1.2>)
    Mesh Scale Mode=::NewEnumerator4
    UsePositionOffset=false
    Write Material Random=true
   )
   &DynamicMaterialParameters(
    Index 0 Param 2=0.12
    Index 0 Param 3=&RandomRangeFloat(Minimum=0, Maximum=1)
    Module.Index0Param0Write=true
    Module.Index0Param1Write=true
    Module.Index0Param2Write=true
    Module.Index0Param3Write=true
   )
   &AddVelocity(
    Velocity=&MergeHelper_VectorSwitch(
     0=&Multiply_Vector(A=$System.DirectionVector, B=<0.5 0.2 0.2>)
     1=$System.DirectionVector
     2=?
     3=?
     Index=$Emitter.SpawnGroup
     Num=::NewEnumerator0
    )
    Velocity Speed Scale=&MergeHelper_FloatSwitch(0=1, 1=5, 2=?, 3=?, Index=$Emitter.SpawnGroup, Num=::NewEnumerator0)
   ) {module=AddVelocity001}
  ]
  particleUpdateScript [
   &ParticleState()
   &ScaleMeshSize(Scale Factor=&MakeVector(
     X=&FloatFromCurve(
      FloatCurve=DataInterface() {propertyValues {Curve {keys [
          {arriveTangent=6.5629, interpMode=RCIM_Cubic, leaveTangent=6.5629, tangentMode=RCTM_User, time=0, value=0.5}
          {arriveTangent=0.5237, interpMode=RCIM_Cubic, leaveTangent=0.5237, tangentMode=RCTM_User, time=0.2, value=1.75}
          {arriveTangent=0.1793, interpMode=RCIM_Cubic, leaveTangent=0.1793, tangentMode=RCTM_User, time=1, value=2}
         ]}}}
      Scale Curve=&FN_FrontEndFloatModifier(Gameplay=1.1, Front End=0.4)
     )
     Y=1
     Z=1
    ))
   &Drag(Drag=4, Use Linear Drag=true, Use Rotational Drag=true)
   &SolveForcesAndVelocity(Clamp Velocity=false, Limit Acceleration=false)
   &ScaleColor(
    Scale Alpha=&FloatFromCurve(FloatCurve=DataInterface() {propertyValues {Curve {keys [
         {arriveTangent=3.396, interpMode=RCIM_Cubic, leaveTangent=3.396, tangentMode=RCTM_User, time=0, value=0}
         {arriveTangent=0.026, interpMode=RCIM_Cubic, leaveTangent=0.026, tangentMode=RCTM_User, time=0.2, value=1}
         {arriveTangent=-1.25, interpMode=RCIM_Cubic, leaveTangent=-1.25, time=0.5, value=0.3}
         <at 1 0 RCIM_Cubic>
        ]}}})
    Scale RGB=&Multiply_Vector(
     A=&MakeVectorFromLinearColorRGB(Color=&ColorFromCurve(ColorCurve=DataInterface() {propertyValues {
         AlphaCurve {keys [<at 0 1 RCIM_Cubic>,<at 1 0 RCIM_Cubic>]}
         BlueCurve {keys [<at 0 0 RCIM_Cubic>,<at 1 0 RCIM_Cubic>]}
         GreenCurve {keys [<at 0 0.4464 RCIM_Cubic>,<at 1 0.2604 RCIM_Cubic>]}
         RedCurve {keys [<at 0 0.4464 RCIM_Cubic>,<at 1 0.2841 RCIM_Cubic>]}
        }}))
     B=&MakeVectorFromLinearColorRGB(Color=$System.TODColorMultiplier)
    )
    ScaleA=true
    ScaleRGB=true
   )
   &FN_Sprite_Attributes(
    Camera Offset=30
    Color=&Multiply_LinearColorByFloat(
     Linear Color=&MultiplyLinearColor(
      Linear Color=&MergeHelper_ColorSwitch(
       0=<rgba 2.5 0 15 1>
       1=<rgba 15 1.4348 5.9565 1>
       2=<rgba 5 2.3623 0.4783 1>
       3=<rgba 0.4783 5 5 1>
       4=<rgba 4.2464 0.4783 5 1>
       5=<rgba 2.1362 0.4783 5 1>
       Index=&MakeIntFromFloat(Float=$User.Variant)
       Num=::NewEnumerator4
      )
      Scale Factor=&MakeVector4_FromLinearColor(LinearColor=$System.TODColorMultiplier)
     )
     Float=&Multiply_Float(
      A=&MakeFloatFromLinearColor(LinearColor=$System.TODColorMultiplier, Channel:ENiagaraLinearColor_Channels=::NewEnumerator3)
      B=&FloatFromCurve(
       FloatCurve=DataInterface() {propertyValues {Curve {keys [{interpMode=RCIM_Cubic, tangentMode=RCTM_User, time=0, value=1}, {interpMode=RCIM_Cubic, tangentMode=RCTM_User, time=1, value=0}]}}}
       Scale Curve=0.75
      )
     )
    )
    Create Camera Offset Attribute=true
    Create Color Attribute=true
    Create SpriteSize Attribute=true
    SpriteSize=&Multiply_Vector2D(
     A=&Multiply_Vector2D_ByFloat(Float=&RandomRangeFloat(Minimum=0.95, Maximum=1.05), Vector2D=<15 300>)
     B=&Vector2DFromCurve(
      Vector2Curve=DataInterface() {propertyValues {
        XCurve {keys [{tangentMode=RCTM_User, time=0, value=0}, <at 0.2 1.075>, {tangentMode=RCTM_User, time=1, value=-0.1}]}
        YCurve {keys [{tangentMode=RCTM_User, time=0, value=0}, <at 0.1 1.025>, {tangentMode=RCTM_User, time=1, value=0.8}]}
       }}
      ScaleCurve=<1 1>
     )
    )
    Write to Module Namespace=true
   )
   &FN_Mesh_Attributes(
    Color=&MakeLinearColorFromVectorAndFloat("Vector (RGB)"=&MakeVectorFromLinearColorRGB(Color=<rgba 0.2552 0.2552 0 1>), "Float (Alpha)"=4)
    Create Color Attribute=true
    Create Scale Attribute=true
    Scale=&Multiply_VectorByFloat(Vector=$Particles.Initial.Scale, Float=1)
    Write to Module Namespace=true
   ) {enabled=false}
  ]
  rendererClasses [&NiagaraMeshRendererProperties]
  renderers [
   {rendererClass=&NiagaraMeshRendererProperties, rendererIndex=0}
   {rendererClass=&NiagaraMeshRendererProperties, rendererIndex=1}
  ]
 }
}
summary {
 systemName=NS_Pickaxe_PointSmoke_Impact_Directional
 emitters [
  {
   emitterName=Flash_Elements
   rendererClasses [
    &NiagaraSpriteRendererProperties
    &NiagaraLightRendererProperties
    &NiagaraMeshRendererProperties
   ]
  }
  {emitterName=Impact_Smoke, rendererClasses [&NiagaraSpriteRendererProperties]}
  {
   emitterName=Directional_Elements
   rendererClasses [&NiagaraSpriteRendererProperties, &NiagaraMeshRendererProperties]
  }
  {emitterName=Sparks, rendererClasses [&NiagaraSpriteRendererProperties]}
  {emitterName=Directional_Streaks001, rendererClasses [&NiagaraSpriteRendererProperties]}
  {emitterName=Toxic_Directional_Smoke, rendererClasses [&NiagaraMeshRendererProperties]}
  {emitterName=CutSkulls001, rendererClasses [&NiagaraSpriteRendererProperties]}
 ]
 userVariables [
  User.bIsFrontEnd {type=NiagaraBool}
  User.bIsFrontEndPreview {type=NiagaraBool}
  User.FrontEndScale {type=NiagaraFloat, defaultValue=0, description=""}
  User.SwingSectionIndex {type=NiagaraInt32}
  User.Variant {type=NiagaraFloat, defaultValue=0, description=""}
 ]
}
