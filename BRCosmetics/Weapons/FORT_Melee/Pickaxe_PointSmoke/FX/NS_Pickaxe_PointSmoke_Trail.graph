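# Niagara trail effect for the PointSmoke pickaxe, with User.Variant choosing among six looks for spawn spacing, size, lifetime and color.
# Elements spawns per unit of movement at a weapon socket, drifting with curl noise and drag, drawn by sprite and ribbon renderers.
# Trails builds layered anim-trail ribbons with skull, tip and galaxy colors faded by TrailFade.
# A time-of-day color multiplier tints both emitters.
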NiagaraSystem /BRCosmetics/Weapons/FORT_Melee/Pickaxe_PointSmoke/FX/NS_Pickaxe_PointSmoke_Trail
dependencies {
 usedDataInterfaces [
  &NiagaraDataInterfaceArrayFloat
  &NiagaraDataInterfaceArrayColor
  &NiagaraDataInterfaceColorCurve
  &NiagaraDataInterfaceCurve
  &NiagaraDataInterfaceSocketReader
 ]
 usedDynamicInputs [
  &SetBoolByIntComparison
  &self:SectionIndexFloatSelector_02
 ]
 usedModules [
  &self:SystemSpawnScript.SystemScriptSource.SystemScriptGraph.NiagaraNodeAssignment_6.SetVariables_#1
  &self:SystemSpawnScript.SystemScriptSource.SystemScriptGraph.NiagaraNodeAssignment_0.SetVariables_#2
  &SystemState
  &self:SystemSpawnScript.SystemScriptSource.SystemScriptGraph.NiagaraNodeAssignment_4.SetVariables_#3
  &self:SystemSpawnScript.SystemScriptSource.SystemScriptGraph.NiagaraNodeAssignment_5.SetVariables_#4
 ]
}
emitters {
 Elements {
  emitterUpdateScript [
   &EmitterState()
   &SpawnPerUnit(
    Max Movement Threshold=4000
    Module.Use Max Movement Threshold=true
    Module.Use Movement Tolerance=false
    Spawn Probability=1
    Spawn Spacing:NiagaraFloat=&SelectFloatFromArray(
     Direct Array Index=&MakeIntFromFloat(Float=$User.Variant)
     Float Selection Array=DataInterface() {propertyValues {FloatData [0,50,50,50,50,30]}}
    )
    Use Spawn Probability=true
   )
  ]
  particleSpawnScript [
   &InitializeParticle(
    Color=&MergeHelper_ColorSwitch(
     0=<rgba 0 0 0 0>
     1=&SelectLinearColorFromArray(
      Array Sampling Mode=::NewEnumerator0
      Direct Array Index=?
      Color Selection Array=DataInterface() {propertyValues {ColorData [<rgba 2.8012 3 0.2698 1>, <rgba 2 2 0 0.4493>]}}
      Random Seed=0
     )
     2=&ColorFromCurve(
      ColorCurve=DataInterface() {propertyValues {
        AlphaCurve {keys [<at 0 1 RCIM_Cubic>]}
        BlueCurve {keys [<at 0.4274 0.1 RCIM_Cubic>,<at 0.4699 0.1 RCIM_Cubic>]}
        GreenCurve {keys [<at 0.4274 0.6609 RCIM_Cubic>,<at 0.4699 0.4283 RCIM_Cubic>]}
        RedCurve {keys [<at 0.4274 1 RCIM_Cubic>,<at 0.4699 1 RCIM_Cubic>]}
       }}
      CurveIndex=&RandomRangeFloat(Minimum=0, Maximum=1)
     )
     3=&ColorFromCurve(
      ColorCurve=DataInterface() {propertyValues {
        AlphaCurve {keys [<at 0 1 RCIM_Cubic>]}
        BlueCurve {keys [<at 0.4274 1 RCIM_Cubic>,<at 0.4699 1 RCIM_Cubic>]}
        GreenCurve {keys [<at 0.4274 0.9469 RCIM_Cubic>,<at 0.4699 0.6167 RCIM_Cubic>]}
        RedCurve {keys [<at 0.4274 0.5208 RCIM_Cubic>,<at 0.4699 0.4375 RCIM_Cubic>]}
       }}
      CurveIndex=&RandomRangeFloat(Minimum=0, Maximum=1)
     )
     4=&ColorFromCurve(
      ColorCurve=DataInterface() {propertyValues {
        AlphaCurve {keys [<at 0 1 RCIM_Cubic>]}
        BlueCurve {keys [<at 0.4274 1 RCIM_Cubic>,<at 0.4699 1 RCIM_Cubic>]}
        GreenCurve {keys [<at 0.4274 0.7649 RCIM_Cubic>,<at 0.4699 0.3664 RCIM_Cubic>]}
        RedCurve {keys [<at 0.4274 0.3886 RCIM_Cubic>,<at 0.4699 0.7831 RCIM_Cubic>]}
       }}
      CurveIndex=&RandomRangeFloat(Minimum=0, Maximum=1)
     )
     5=&ColorFromCurve(
      ColorCurve=DataInterface() {propertyValues {
        AlphaCurve {keys [<at 0 1 RCIM_Cubic>]}
        BlueCurve {keys [<at 0.4274 1 RCIM_Cubic>,<at 0.4699 1 RCIM_Cubic>]}
        GreenCurve {keys [<at 0.4274 0 RCIM_Cubic>,<at 0.4699 0.282 RCIM_Cubic>]}
        RedCurve {keys [<at 0.4274 0.6698 RCIM_Cubic>,<at 0.4699 0 RCIM_Cubic>]}
       }}
      CurveIndex=&RandomRangeFloat(Minimum=0, Maximum=1)
     )
     Index=&MakeIntFromFloat(Float=$User.Variant, Conversion Mode=::NewEnumerator1)
     Num=::NewEnumerator4
    )
    Lifetime=&MergeHelper_FloatSwitch(
     0=0
     1=&RandomRangeFloat(Minimum=0.15, Maximum=0.3)
     2=&RandomRangeFloat(Minimum=0.3, Maximum=0.4)
     3=&RandomRangeFloat(Minimum=0.35, Maximum=0.45)
     4=&RandomRangeFloat(Minimum=0.25, Maximum=0.35)
     5=0.15
     Index=&MakeIntFromFloat(Float=$User.Variant, Conversion Mode=::NewEnumerator1)
     Num=::NewEnumerator4
    )
    Ribbon Width=1
    Ribbon Width Mode=::NewEnumerator1
    Sprite Rotation Angle=&MergeHelper_FloatSwitch(
     0=0
     1=0
     2=0
     3=&RandomRangeFloat(Minimum=0, Maximum=1)
     4=0
     5=0
     Index=&MakeIntFromFloat(Float=$User.Variant, Conversion Mode=::NewEnumerator1)
     Num=::NewEnumerator4
    )
    Sprite Rotation Mode=::NewEnumerator1
    Sprite Size Max=&MergeHelper_Vector2Switch(
     0=<0 0>
     1=<4 3>
     2=<16 16>
     3=<13 13>
     4=<10 10>
     5=<20 20>
     Index=&MakeIntFromFloat(Float=$User.Variant)
     Num=::NewEnumerator4
    )
    Sprite Size Min=&MergeHelper_Vector2Switch(
     0=<0 0>
     1=<2.5 2>
     2=<12 12>
     3=<9 9>
     4=<7 7>
     5=<15 15>
     Index=&MakeIntFromFloat(Float=$User.Variant)
     Num=::NewEnumerator4
    )
    Sprite Size Mode=::NewEnumerator2
    UsePositionOffset=true
    Write Material Random=true
   )
   &SocketLocation(
    Interpolate Socket Sampling=true
    SamplePosOffset=true
    Sampled Position Offset=&RandomRangeVector(Minimum=<0 -5 0>, Maximum=<0 5 0>)
    Socket Index=1
    Socket Reader=$System.Socket Reader
    Socket Sampling Mode=::NewEnumerator3
    VelocitySpeedLimit=false
   )
   &AddVelocity(
    Rotation Coordinate Space=::Simulation
    Velocity=&RotateVectorByQuaternion(Vector To Rotate=<0 0 1>, Quaternion=$Output.SocketLocation.SampledRotation)
    Velocity Speed Scale=&RandomRangeFloat(Minimum=75, Maximum=150)
   ) {module=AddVelocity001}
   &InheritVelocity(Inherited Velocity Speed Limit=25, Speed Limit=true, UseSpeedThreshold=false)
   /SetVariables(Particles.SubImageIndexAlt:NiagaraFloat=&RandomRangeFloat(Minimum=0, Maximum=36)) {isSetParameters=true}
  ]
  particleUpdateScript [
   &ParticleState()
   &SocketLocation(
    Interpolate Socket Sampling=true
    Position Sampling=::NewEnumerator1
    SamplePosOffset=false
    Socket Index=1
    Socket Reader=$System.Socket Reader
    Socket Sampling Mode=::NewEnumerator3
    VelocitySpeedLimit=false
   ) {module=SocketLocation001}
   &ScaleSpriteSize(Uniform Curve Sprite Scale=DataInterface() {propertyValues {Curve {keys [
        {arriveTangent=8.2127, interpMode=RCIM_Cubic, leaveTangent=8.2127, tangentMode=RCTM_User, time=0, value=0}
        {
         arriveTangent=0.0908
         arriveTangentWeight=0.232
         interpMode=RCIM_Cubic
         leaveTangent=0.0908
         leaveTangentWeight=2.2538
         tangentMode=RCTM_User
         time=0.1
         value=1
        }
        {arriveTangent=-0.042, interpMode=RCIM_Cubic, leaveTangent=-0.042, tangentMode=RCTM_User, time=1, value=0}
       ]}}}) {module=ScaleSpriteSize001}
   &Drag(Drag=&RandomRangeFloat(Minimum=6, Maximum=8), Use Linear Drag=true, Use Rotational Drag=true)
   &CurlNoiseForce(
    Noise Strength=&FloatFromCurve(
     FloatCurve=DataInterface() {propertyValues {Curve {keys [
         {arriveTangent=2.7279, interpMode=RCIM_Cubic, leaveTangent=2.7279, tangentMode=RCTM_User, time=0, value=0}
         {interpMode=RCIM_Cubic, tangentMode=RCTM_User, time=1, value=1}
        ]}}}
     Scale Curve=550
    )
    Pan Noise=false
   )
   &ScaleRibbonWidth(Ribbon Width Scale=&FloatFromCurve(FloatCurve=DataInterface() {propertyValues {Curve {keys [{interpMode=RCIM_Cubic, tangentMode=RCTM_User, time=0, value=1}, {interpMode=RCIM_Cubic, tangentMode=RCTM_User, time=1, value=0}]}}}))
   &SolveForcesAndVelocity(Clamp Velocity=false, Limit Acceleration=false)
   &ScaleSpriteSizeBySpeed(
    Max Scale Factor=&MergeHelper_Vector2Switch(
     0=<1 1>
     1=<1 5>
     2=<1 1>
     3=<1 1>
     4=<1 1>
     5=<1 1>
     Index=&MakeIntFromFloat(Float=$User.Variant)
     Num=::NewEnumerator4
    )
    Min Scale Factor=&MergeHelper_Vector2Switch(
     0=<1 1>
     1=<0.5 0.9>
     2=<1 1>
     3=<1 1>
     4=<1 1>
     5=<1 1>
     Index=&MakeIntFromFloat(Float=$User.Variant)
     Num=::NewEnumerator4
    )
    Velocity Threshold=250
   )
   &ScaleColor(
    Scale Alpha=&Multiply_Float(
     A=&MakeFloatFromLinearColor(LinearColor=$System.TODColorMultiplier, Channel:ENiagaraLinearColor_Channels=::NewEnumerator3)
     B=1
    )
    Scale RGB=&MakeVectorFromLinearColorRGB(Color=$System.TODColorMultiplier)
    ScaleA=true
    ScaleRGB=true
   )
   /SetVariables(
    Particles.GlowColor=&MakeLinearColorFromVectorAndFloat(
     "Vector (RGB)"=&MakeVectorFromLinearColorRGB(Color=$Particles.Color)
     "Float (Alpha)"=&FloatFromCurve(FloatCurve=DataInterface() {propertyValues {Curve {keys [<at 0.5 1>,<at 1 0>]}}}, Scale Curve=0.08)
    )
    Particles.GlowSize=&Multiply_Vector2D(A=$Particles.SpriteSize, B=<8 2>)
   ) {isSetParameters=true}
  ]
  rendererClasses [&NiagaraSpriteRendererProperties, &NiagaraRibbonRendererProperties]
  renderers [
   {rendererClass=&NiagaraSpriteRendererProperties}
   {rendererClass=&NiagaraSpriteRendererProperties, rendererIndex=1}
   {rendererClass=&NiagaraSpriteRendererProperties, rendererIndex=2}
   {rendererClass=&NiagaraSpriteRendererProperties, rendererIndex=3}
   {rendererClass=&NiagaraRibbonRendererProperties, rendererIndex=4}
   {rendererClass=&NiagaraSpriteRendererProperties, rendererIndex=5}
   {rendererClass=&NiagaraSpriteRendererProperties, rendererIndex=6}
  ]
 }
 Trails {
  emitterSpawnScript [&FN_RibbonIDOffsetUtility()]
  emitterUpdateScript [
   &EmitterState()
   &FN_SpawnAnimTrail(Spawn Interval=0)
  ]
  particleSpawnScript [
   &FN_InitializeAnimTrailBoneIndecies()
   &FN_InitializeAnimTrail(
    Color=&MergeHelper_ColorSwitch(
     0=<rgba 0 0 0 2>
     1=<rgba 0 0 0 2>
     2=<rgba 0 0 0 2>
     3=<rgba 0.0909 0.5645 0.7652 0.9014>
     4=<rgba 1 1 1 1>
     5=<rgba 0.0259 0.0392 0.0608 2>
     Index=&MakeIntFromFloat(Float=$User.Variant)
     Num=::NewEnumerator4
    )
    Lifetime=0.3
   )
   /SetVariables(
    Particles.Skull Trail Color:LinearColor=&MultiplyLinearColor(
     Linear Color=&MergeHelper_ColorSwitch(
      0=<rgba 0.75 0.75 0.75 1>
      1=<rgba 0.44 0.5 0.05 2>
      2=<rgba 0.5 0.4063 0.125 1>
      3=<rgba 0.5543 0.6918 0.75 1>
      4=<rgba 0.5 0.5 0.5 1>
      5=<rgba 0.6609 0.6609 0.6609 1>
      Index=&MakeIntFromFloat(Float=$User.Variant)
      Num=::NewEnumerator4
     )
     Scale Factor=&MakeVector4_FromLinearColor(LinearColor=$System.TODColorMultiplier)
    )
    Particles.Skull Trail Lifetime=0.3
   ) {isSetParameters=true}
   /SetVariables(Particles.Tip Trail Lifetime=0.3) {isSetParameters=true}
   /SetVariables(
    Particles.GalaxyDynPar:Vector4f=&MakeVector4(X=0, Y=1, Z=0, W=0.3)
    Particles.GalaxyDynPar1:Vector4f=&MakeVector4(X=0, Y=0, Z=0, W=0)
   ) {isSetParameters=true}
   /SetVariables(
    Particles.Galaxy Trail Color:LinearColor=&MultiplyLinearColor(
     Linear Color=<rgba 1 1 1 1>
     Scale Factor=&MakeVector4_FromLinearColor(LinearColor=$System.TODColorMultiplier)
    )
    Particles.Galaxy Trail Lifetime=0.3
   ) {isSetParameters=true}
  ]
  particleUpdateScript [
   &FN_UpdateAnimTrail() {module=FN_UpdateAnimTrail001}
   /SetVariables(
    Particles.Update Galaxy Trail Lifetime:NiagaraFloat=&Divide_Float(A=$Particles.Age, B=$Particles.Initial.Galaxy Trail Lifetime)
    Particles.Update Skull Trail Lifetime:NiagaraFloat=&Divide_Float(A=$Particles.Age, B=$Particles.Initial.Skull Trail Lifetime)
    Particles.Update Tip Trail Lifetime:NiagaraFloat=&Divide_Float(A=$Particles.Age, B=$Particles.Initial.Tip Trail Lifetime)
   ) {isSetParameters=true}
   /SetVariables(
    Particles.Update Galaxy Trail Color:LinearColor=&MakeLinearColorFromVectorAndFloat(
     "Vector (RGB)"=&MakeVectorFromLinearColorRGB(Color=<rgba 1 1 1 1>)
     "Float (Alpha)"=&FloatFromCurve(
      FloatCurve=DataInterface() {propertyValues {Curve {keys [
          {arriveTangent=0.1035, interpMode=RCIM_Cubic, leaveTangent=0.1035, tangentMode=RCTM_User, time=0, value=1}
          {arriveTangent=-2.1259, interpMode=RCIM_Cubic, leaveTangent=-2.1259, tangentMode=RCTM_User, time=1, value=0}
         ]}}}
      CurveIndex=$Particles.Update Galaxy Trail Lifetime
      Scale Curve=&Multiply_Float(
       A=&FNTimeOfDayScalar(Night Scale=2, Morning Scale=2, Day Scale=2, Evening Scale=2)
       B=$System.TrailFade
      )
     )
    )
    Particles.Update Skull Trail Color:LinearColor=&MakeLinearColorFromVectorAndFloat(
     "Vector (RGB)"=&Multiply_Vector(
      A=&Multiply_Vector(
       A=&MakeVectorFromLinearColorRGB(Color=$Particles.Initial.Skull Trail Color)
       B=&MakeVectorFromLinearColorRGB(Color=&MergeHelper_ColorSwitch(
         0=<rgba 1 1 1 1>
         1=<rgba 1 1 1 1>
         2=<rgba 1 1 1 1>
         3=<rgba 1 1 1 1>
         4=&ColorFromCurve(ColorCurve=DataInterface() {propertyValues {
            AlphaCurve {keys []}
            BlueCurve {keys [<at 0.1197 1>,<at 0.2295 0.7014>,<at 0.2879 0.4393>,<at 0.3572 0.429>,<at 0.4592 0.429>,<at 0.7216 0.429>,<at 0.8836 0.429>]}
            GreenCurve {keys [<at 0.1197 0.3852>,<at 0.2295 0.1586>,<at 0.2879 0.2321>,<at 0.3572 0.0993>,<at 0.4592 0.2663>,<at 0.7216 0.1328>,<at 0.8836 0.2422>]}
            RedCurve {keys [<at 0.1197 0.2375>,<at 0.2295 0.611>,<at 0.2879 0.0882>,<at 0.3572 0.2847>,<at 0.4592 0.1656>,<at 0.7216 0.2905>,<at 0.8836 0.09>]}
           }})
         5=&ColorFromCurve(ColorCurve=DataInterface() {propertyValues {
            AlphaCurve {keys []}
            BlueCurve {keys [<at 0.0053 1>, {arriveTangent=-0.4939, interpMode=RCIM_Cubic, leaveTangent=-0.4939, time=0.1649, value=0.467}, <at 1 0.5087 RCIM_Cubic>]}
            GreenCurve {keys [<at 0.0053 0.7729>, {arriveTangent=-0.4393, interpMode=RCIM_Cubic, leaveTangent=-0.4393, time=0.1649, value=0.361}, <at 1 0.336 RCIM_Cubic>]}
            RedCurve {keys [
              <at 0.0053 0.3304>
              {arriveTangent=-0.1811, interpMode=RCIM_Cubic, leaveTangent=-0.1811, time=0.1649, value=0.1543}
              <at 1 0.1502 RCIM_Cubic>
             ]}
           }})
         Index=&MakeIntFromFloat(Float=$User.Variant)
         Num=::NewEnumerator4
        ))
      )
      B=&MakeVectorFromLinearColorRGB(Color=$System.TODColorMultiplier)
     )
     "Float (Alpha)"=&Multiply_Float(
      A=&MakeFloatFromLinearColor(LinearColor=$System.TODColorMultiplier, Channel:ENiagaraLinearColor_Channels=::NewEnumerator3)
      B=&Multiply_Float(
       A=&FloatFromCurve(
        FloatCurve=DataInterface() {propertyValues {Curve {keys [
            {arriveTangent=-1, interpMode=RCIM_Cubic, time=0.12, value=1}
            {arriveTangent=-0.1576, interpMode=RCIM_Cubic, tangentMode=RCTM_Break, time=1, value=0}
           ]}}}
        Scale Curve=&MergeHelper_FloatSwitch(0=1, 1=1, 2=1, 3=1, 4=2, 5=0.75, Index=&MakeIntFromFloat(Float=$User.Variant), Num=::NewEnumerator4)
       )
       B=$System.TrailFade
      )
     )
    )
    Particles.Update Tip Trail Color:LinearColor=&MakeLinearColorFromVectorAndFloat(
     "Vector (RGB)"=&Multiply_Vector(
      A=&MakeVectorFromLinearColorRGB(Color=&MergeHelper_ColorSwitch(
        0=&ColorFromCurve(
         ColorCurve=DataInterface() {propertyValues {
           AlphaCurve {keys []}
           BlueCurve {keys [<at 0 0.4948>,<at 0.5969 0.5208>]}
           GreenCurve {keys [<at 0 0.4948>,<at 0.5969 0.5208>]}
           RedCurve {keys [<at 0 0.4948>,<at 0.5969 0.5208>]}
          }}
         CurveIndex=$Particles.Update Tip Trail Lifetime
        )
        1=&ColorFromCurve(
         ColorCurve=DataInterface() {propertyValues {
           AlphaCurve {keys []}
           BlueCurve {keys [<at 0 0.2822>,<at 0.5969 0.1306>]}
           GreenCurve {keys [<at 0 0.3738>,<at 0.5969 0.1509>]}
           RedCurve {keys [<at 0 0.4406>,<at 0.5969 0.1739>]}
          }}
         CurveIndex=$Particles.Update Tip Trail Lifetime
        )
        2=&ColorFromCurve(
         ColorCurve=DataInterface() {propertyValues {
           AlphaCurve {keys []}
           BlueCurve {keys [<at 0 0.3>,<at 0.5969 0.1652>]}
           GreenCurve {keys [<at 0 0.6657>,<at 0.5969 0.4406>]}
           RedCurve {keys [<at 0 1>,<at 0.5969 0.8261>]}
          }}
         CurveIndex=$Particles.Update Tip Trail Lifetime
        )
        3=&ColorFromCurve(
         ColorCurve=DataInterface() {propertyValues {
           AlphaCurve {keys []}
           BlueCurve {keys [<at 0 0.6476>,<at 0.5969 0.5159>]}
           GreenCurve {keys [<at 0 0.5964>,<at 0.5969 0.4487>]}
           RedCurve {keys [<at 0 0.5462>,<at 0.5969 0.4113>]}
          }}
         CurveIndex=$Particles.Update Tip Trail Lifetime
        )
        4=&ColorFromCurve(
         ColorCurve=DataInterface() {propertyValues {
           AlphaCurve {keys []}
           BlueCurve {keys [<at 0 1>,<at 0.2628 0.9093>,<at 0.5969 2>]}
           GreenCurve {keys [<at 0 0.3852>,<at 0.2628 1>,<at 0.5969 0.4522>]}
           RedCurve {keys [<at 0 0.2375>,<at 0.2628 0.313>,<at 0.5969 1.742>]}
          }}
         CurveIndex=$Particles.Update Tip Trail Lifetime
        )
        5=&ColorFromCurve(
         ColorCurve=DataInterface() {propertyValues {
           AlphaCurve {keys []}
           BlueCurve {keys [<at 0 1>,<at 0.5969 0.395>]}
           GreenCurve {keys [<at 0 0.696>,<at 0.5969 0>]}
           RedCurve {keys [<at 0 0.1772>,<at 0.5969 0.0593>]}
          }}
         CurveIndex=$Particles.Update Tip Trail Lifetime
        )
        Index=&MakeIntFromFloat(Float=$User.Variant)
        Num=::NewEnumerator4
       ))
      B=&MakeVectorFromLinearColorRGB(Color=$System.TODColorMultiplier)
     )
     "Float (Alpha)"=&Multiply_Float(
      A=&MakeFloatFromLinearColor(LinearColor=$System.TODColorMultiplier, Channel:ENiagaraLinearColor_Channels=::NewEnumerator3)
      B=&Multiply_Float(
       A=&FloatFromCurve(
        FloatCurve=DataInterface() {propertyValues {Curve {keys [
            {arriveTangent=-1, interpMode=RCIM_Cubic, time=0.12, value=1}
            {arriveTangent=-0.1576, interpMode=RCIM_Cubic, tangentMode=RCTM_Break, time=1, value=0}
           ]}}}
        Scale Curve=&MergeHelper_FloatSwitch(
         0=1.5
         1=2.5
         2=1.5
         3=2
         4=1
         5=0.75
         Index=&MakeIntFromFloat(Float=$User.Variant)
         Num=::NewEnumerator4
        )
       )
       B=$System.TrailFade
      )
     )
    )
   ) {isSetParameters=true}
   &ScaleColor(
    Scale Alpha=&Multiply_Float(
     A=&MakeFloatFromLinearColor(LinearColor=$System.TODColorMultiplier, Channel:ENiagaraLinearColor_Channels=::NewEnumerator3)
     B=&Multiply_Float(
      A=&FloatFromCurve(
       FloatCurve=DataInterface() {propertyValues {Curve {keys [
           {arriveTangent=-1, interpMode=RCIM_Cubic, time=0.12, value=1}
           {arriveTangent=-0.1576, interpMode=RCIM_Cubic, tangentMode=RCTM_Break, time=1, value=0}
          ]}}}
       Scale Curve=&MergeHelper_FloatSwitch(0=1, 1=1, 2=1, 3=1, 4=0.75, 5=1.25, Index=&MakeIntFromFloat(Float=$User.Variant), Num=::NewEnumerator4)
      )
      B=$System.TrailFade
     )
    )
    Scale RGB=&Multiply_Vector(
     A=&MakeVectorFromLinearColorRGB(Color=&MergeHelper_ColorSwitch(
       0=<rgba 1 1 1 1>
       1=<rgba 1 1 1 1>
       2=<rgba 1 1 1 1>
       3=<rgba 1 1 1 1>
       4=&ColorFromCurve(ColorCurve=DataInterface() {propertyValues {
          AlphaCurve {keys []}
          BlueCurve {keys [<at 0.0734 1>,<at 0.1327 1>,<at 0.1942 1>,<at 0.291 1>,<at 0.4156 0.429>,<at 0.5267 0.2261>]}
          GreenCurve {keys [<at 0.0734 0.3852>,<at 0.1327 0.2261>,<at 0.1942 0.8117>,<at 0.291 0.6849>,<at 0.4156 0.0993>,<at 0.5267 0.1345>]}
          RedCurve {keys [<at 0.0734 0.2375>,<at 0.1327 0.871>,<at 0.1942 0.2617>,<at 0.291 0.3299>,<at 0.4156 0.2847>,<at 0.5267 0.0664>]}
         }})
       5=<rgba 1 1 1 1>
       Index=&MakeIntFromFloat(Float=$User.Variant)
       Num=::NewEnumerator4
      ))
     B=&MakeVectorFromLinearColorRGB(Color=$System.TODColorMultiplier)
    )
    ScaleA=true
    ScaleRGB=true
   ) {module=ScaleColor002}
  ]
  rendererClasses [&NiagaraRibbonRendererProperties]
  renderers [
   {rendererClass=&NiagaraRibbonRendererProperties}
   {rendererClass=&NiagaraRibbonRendererProperties, rendererIndex=1}
   {rendererClass=&NiagaraRibbonRendererProperties, rendererIndex=2}
   {rendererClass=&NiagaraRibbonRendererProperties, rendererIndex=3}
   {rendererClass=&NiagaraRibbonRendererProperties, rendererIndex=4}
  ]
 }
}
summary {
 systemName=NS_Pickaxe_PointSmoke_Trail
 emitters [
  {emitterName=Elements, rendererClasses [&NiagaraSpriteRendererProperties, &NiagaraRibbonRendererProperties]}
  {emitterName=Trails, rendererClasses [&NiagaraRibbonRendererProperties]}
 ]
 userVariables [
  User.bIsFrontEnd {type=NiagaraBool}
  User.bIsFrontEndPreview {type=NiagaraBool}
  User.RibbonIDOffset {type=NiagaraInt32}
  User.SwingSectionIndex {type=NiagaraInt32}
  User.Variant {type=NiagaraFloat, defaultValue=0, description=""}
 ]
}
